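# Widget Blueprint for the HUD Layout Tool V2 profile selection screen, where players pick and manage preset and custom HUD layouts.
# A scroll box holds wrapping entry boxes of layout buttons for both groups, with empty-slot placeholders in an overlay beneath the custom layouts.
# A bottom bar offers delete, rename, create, legacy-convert and select buttons; Import and Share are collapsed.
# The event graph is empty, and key widgets are inherited from the native parent class.
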WidgetBlueprint /FortniteCoreUI/Athena/HUD/Bacchus/HUDLayoutTool/HUDLayoutToolV2_ProfileSelectionPanel
graphs [EventGraph {links [], nodes []}]
variables []
widgets {
 info {
  inheritedWidgetCount=9
  parentClass=/Script/FortniteUI.HUDLayoutToolV2_ProfileSelectionPanel
  rootWidgetClass=&Overlay
  widgetCount=27
 }
 widgets [
  Overlay_67 {widgetClassPath=&Overlay, widget=Overlay_67}
  Background {
   widgetClassPath=&Image
   parent=Overlay_67
   widget=Background
   properties {brush {
     bIsDynamicallyLoaded=false
     drawAs=Image
     imageType=NoImage
     mirroring=NoMirror
     resourceName=None
     resourceObject=/Game/Athena/HUD/Bacchus/HUDLayoutTool/HLT_Icons/M_HLT_RadialGradient_Inst
     tiling=NoTile
     imageSize=<32 32>
     margin=<ltrb 0 0 0 0>
     outlineSettings {
      bUseBrushTransparency=false
      roundingType=HalfHeightRadius
      width=0
      color {colorUseRule=UseColor_Specified, specifiedColor=<rgba 0 0 0 0>}
      cornerRadii=<0 0 0 0>
     }
     tintColor {colorUseRule=UseColor_Specified, specifiedColor=<rgba 1 1 1 1>}
     uVRegion {bIsValid=false, max=<0 0>, min=<0 0>}
    }}
   slot {class=OverlaySlot, properties {horizontalAlignment=HAlign_Fill, verticalAlignment=VAlign_Fill}}
  }
  SizeBox_0 {
   widgetClassPath=&SizeBox
   parent=Overlay_67
   widget=SizeBox_0
   properties {bOverride_HeightOverride=true, heightOverride=160}
   slot {class=OverlaySlot, properties {horizontalAlignment=HAlign_Fill}}
  }
  Header_Bottom {
   widgetClassPath=&Image
   parent=SizeBox_0
   widget=Header_Bottom
   properties {
    brush {
     bIsDynamicallyLoaded=false
     drawAs=Image
     imageType=NoImage
     mirroring=NoMirror
     resourceName=None
     resourceObject=/Game/Athena/HUD/Bacchus/HUDLayoutTool/HLT_Icons/M_HLT_ManagementHeader
     tiling=NoTile
     imageSize=<32 32>
     margin=<ltrb 0 0 0 0>
     outlineSettings {
      bUseBrushTransparency=false
      roundingType=HalfHeightRadius
      width=0
      color {colorUseRule=UseColor_Specified, specifiedColor=<rgba 0 0 0 0>}
      cornerRadii=<0 0 0 0>
     }
     tintColor {colorUseRule=UseColor_Specified, specifiedColor=<rgba 1 1 1 1>}
     uVRegion {bIsValid=false, max=<0 0>, min=<0 0>}
    }
    renderTransform {angle=0, scale=<-1 1>, shear=<0 0>, translation=<0 0>}
   }
   slot {class=SizeBoxSlot}
  }
  SafeZone_3 {
   widgetClassPath=&SafeZone
   bIsVariable=true
   parent=Overlay_67
   widget=SafeZone_3
   slot {class=OverlaySlot, properties {horizontalAlignment=HAlign_Fill, verticalAlignment=VAlign_Fill}}
  }
  Overlay_177 {widgetClassPath=&Overlay, parent=SafeZone_3, widget=Overlay_177, slot {class=SafeZoneSlot}}
  CommonTextBlock_87 {
   widgetClassPath=&CommonTextBlock
   parent=Overlay_177
   widget=CommonTextBlock_87
   properties {
    text=HUD Layouts
    textTransformPolicy=ToUpper
    font {
     bForceMonospaced=false
     bMaterialIsStencil=false
     fontMaterial=null
     fontObject=&BurbankSmall
     letterSpacing=0
     monospacedWidth=1
     size=50
     skewAmount=0
     typefaceFontName=Black
     outlineSettings {
      bApplyOutlineToDropShadows=false
      bMiteredCorners=false
      bSeparateFillAlpha=false
      outlineMaterial=null
      outlineSize=0
      outlineColor=<rgba 0 0 0 1>
     }
    }
    renderTransform {angle=0, scale=<1 1>, shear=<-10 0>, translation=<0 0>}
    shadowOffset=<0 0>
   }
   slot {class=OverlaySlot, properties {padding=<ltrb 100 40 0 0>}}
  }
  Button_Exit {
   widgetClassPath=/FortUILibrary/Foundation/Buttons/CloseButton
   bInherited=true
   parent=Overlay_177
   widget=Button_Exit
   properties {
    onButtonBaseClicked=(&self:WidgetTree.Button_Exit.HandleOnButtonClicked)
    onButtonBaseDoubleClicked=(&self:WidgetTree.Button_Exit.HandleOnButtonDoubleClicked)
    onButtonBaseHovered=(&self:WidgetTree.Button_Exit.HandleOnButtonHovered)
    onButtonBaseUnhovered=(&self:WidgetTree.Button_Exit.HandleOnButtonUnhovered)
    onSelectedChangedBase=(&self:WidgetTree.Button_Exit.HandleOnSelectedChanged)
   }
   slot {class=OverlaySlot, properties {horizontalAlignment=HAlign_Right, padding=<ltrb 0 20 20 0>}}
  }
  CommonHierarchicalScrollBox_73 {
   widgetClassPath=/Script/CommonUI.CommonHierarchicalScrollBox
   parent=Overlay_177
   widget=CommonHierarchicalScrollBox_73
   properties {
    bIsVolatile=true
    scrollbarPadding=<ltrb 0 0 0 0>
    scrollbarThickness=<20 9>
    widgetBarStyle {
     thickness=16
     draggedThumbImage {
      bIsDynamicallyLoaded=false
      drawAs=Box
      imageType=FullColor
      mirroring=NoMirror
      resourceName=../../../Engine/Content/Slate/Common/Scrollbar_Thumb.png
      resourceObject=null
      tiling=NoTile
      imageSize=<0 0>
      margin=<ltrb 0.25 0.25 0.25 0.25>
      outlineSettings {
       bUseBrushTransparency=false
       roundingType=HalfHeightRadius
       width=0
       color {colorUseRule=UseColor_Specified, specifiedColor=<rgba 0 0 0 0>}
       cornerRadii=<0 0 0 0>
      }
      tintColor {colorUseRule=UseColor_Specified, specifiedColor=<rgba 1 1 1 1>}
      uVRegion {bIsValid=false, max=<0 0>, min=<0 0>}
     }
     horizontalBackgroundImage {
      bIsDynamicallyLoaded=false
      drawAs=NoDrawType
      imageType=NoImage
      mirroring=NoMirror
      resourceName=None
      resourceObject=null
      tiling=NoTile
      imageSize=<0 0>
      margin=<ltrb 0 0 0 0>
      outlineSettings {
       bUseBrushTransparency=false
       roundingType=HalfHeightRadius
       width=0
       color {colorUseRule=UseColor_Specified, specifiedColor=<rgba 0 0 0 0>}
       cornerRadii=<0 0 0 0>
      }
      tintColor {colorUseRule=UseColor_Specified, specifiedColor=<rgba 1 1 1 1>}
      uVRegion {bIsValid=false, max=<0 0>, min=<0 0>}
     }
     horizontalBottomSlotImage {
      bIsDynamicallyLoaded=false
      drawAs=NoDrawType
      imageType=FullColor
      mirroring=NoMirror
      resourceName=../../../Engine/Content/Slate/Common/Scrollbar_Background_Horizontal.png
      resourceObject=null
      tiling=NoTile
      imageSize=<8 8>
      margin=<ltrb 0 0 0 0>
      outlineSettings {
       bUseBrushTransparency=false
       roundingType=HalfHeightRadius
       width=0
       color {colorUseRule=UseColor_Specified, specifiedColor=<rgba 0 0 0 0>}
       cornerRadii=<0 0 0 0>
      }
      tintColor {colorUseRule=UseColor_Specified, specifiedColor=<rgba 1 1 1 1>}
      uVRegion {bIsValid=false, max=<0 0>, min=<0 0>}
     }
     horizontalTopSlotImage {
      bIsDynamicallyLoaded=false
      drawAs=NoDrawType
      imageType=FullColor
      mirroring=NoMirror
      resourceName=../../../Engine/Content/Slate/Common/Scrollbar_Background_Horizontal.png
      resourceObject=null
      tiling=NoTile
      imageSize=<8 8>
      margin=<ltrb 0 0 0 0>
      outlineSettings {
       bUseBrushTransparency=false
       roundingType=HalfHeightRadius
       width=0
       color {colorUseRule=UseColor_Specified, specifiedColor=<rgba 0 0 0 0>}
       cornerRadii=<0 0 0 0>
      }
      tintColor {colorUseRule=UseColor_Specified, specifiedColor=<rgba 1 1 1 1>}
      uVRegion {bIsValid=false, max=<0 0>, min=<0 0>}
     }
     hoveredThumbImage {
      bIsDynamicallyLoaded=false
      drawAs=Box
      imageType=FullColor
      mirroring=NoMirror
      resourceName=../../../Engine/Content/Slate/Common/Scrollbar_Thumb.png
      resourceObject=null
      tiling=NoTile
      imageSize=<0 0>
      margin=<ltrb 0.25 0.25 0.25 0.25>
      outlineSettings {
       bUseBrushTransparency=false
       roundingType=HalfHeightRadius
       width=0
       color {colorUseRule=UseColor_Specified, specifiedColor=<rgba 0 0 0 0>}
       cornerRadii=<0 0 0 0>
      }
      tintColor {colorUseRule=UseColor_Specified, specifiedColor=<rgba 1 1 1 1>}
      uVRegion {bIsValid=false, max=<0 0>, min=<0 0>}
     }
     normalThumbImage {
      bIsDynamicallyLoaded=false
      drawAs=Box
      imageType=FullColor
      mirroring=NoMirror
      resourceName=../../../Engine/Content/Slate/Common/Scrollbar_Thumb.png
      resourceObject=null
      tiling=NoTile
      imageSize=<0 0>
      margin=<ltrb 0.25 0.25 0.25 0.25>
      outlineSettings {
       bUseBrushTransparency=false
       roundingType=HalfHeightRadius
       width=0
       color {colorUseRule=UseColor_Specified, specifiedColor=<rgba 0 0 0 0>}
       cornerRadii=<0 0 0 0>
      }
      tintColor {colorUseRule=UseColor_Specified, specifiedColor=<rgba 0 0.2644 1 1>}
      uVRegion {bIsValid=false, max=<0 0>, min=<0 0>}
     }
     verticalBackgroundImage {
      bIsDynamicallyLoaded=false
      drawAs=Image
      imageType=NoImage
      mirroring=NoMirror
      resourceName=None
      resourceObject=null
      tiling=NoTile
      imageSize=<0 0>
      margin=<ltrb 0 0 0 0>
      outlineSettings {
       bUseBrushTransparency=false
       roundingType=HalfHeightRadius
       width=0
       color {colorUseRule=UseColor_Specified, specifiedColor=<rgba 0 0 0 0>}
       cornerRadii=<0 0 0 0>
      }
      tintColor {colorUseRule=UseColor_Specified, specifiedColor=<rgba 0.0056 0.0222 0.3005 1>}
      uVRegion {bIsValid=false, max=<0 0>, min=<0 0>}
     }
     verticalBottomSlotImage {
      bIsDynamicallyLoaded=false
      drawAs=NoDrawType
      imageType=FullColor
      mirroring=NoMirror
      resourceName=../../../Engine/Content/Slate/Common/Scrollbar_Background_Vertical.png
      resourceObject=null
      tiling=NoTile
      imageSize=<8 8>
      margin=<ltrb 0 0 0 0>
      outlineSettings {
       bUseBrushTransparency=false
       roundingType=HalfHeightRadius
       width=0
       color {colorUseRule=UseColor_Specified, specifiedColor=<rgba 0 0 0 0>}
       cornerRadii=<0 0 0 0>
      }
      tintColor {colorUseRule=UseColor_Specified, specifiedColor=<rgba 1 1 1 1>}
      uVRegion {bIsValid=false, max=<0 0>, min=<0 0>}
     }
     verticalTopSlotImage {
      bIsDynamicallyLoaded=false
      drawAs=NoDrawType
      imageType=FullColor
      mirroring=NoMirror
      resourceName=../../../Engine/Content/Slate/Common/Scrollbar_Background_Vertical.png
      resourceObject=null
      tiling=NoTile
      imageSize=<8 8>
      margin=<ltrb 0 0 0 0>
      outlineSettings {
       bUseBrushTransparency=false
       roundingType=HalfHeightRadius
       width=0
       color {colorUseRule=UseColor_Specified, specifiedColor=<rgba 0 0 0 0>}
       cornerRadii=<0 0 0 0>
      }
      tintColor {colorUseRule=UseColor_Specified, specifiedColor=<rgba 1 1 1 1>}
      uVRegion {bIsValid=false, max=<0 0>, min=<0 0>}
     }
    }
    widgetStyle {
     barThickness=9
     bottomShadowBrush {
      bIsDynamicallyLoaded=false
      drawAs=NoDrawType
      imageType=FullColor
      mirroring=NoMirror
      resourceName=../../../Engine/Content/Slate/Common/ScrollBoxShadowBottom.png
      resourceObject=null
      tiling=NoTile
      imageSize=<16 8>
      margin=<ltrb 0.5 0 0.5 1>
      outlineSettings {
       bUseBrushTransparency=false
       roundingType=HalfHeightRadius
       width=0
       color {colorUseRule=UseColor_Specified, specifiedColor=<rgba 0 0 0 0>}
       cornerRadii=<0 0 0 0>
      }
      tintColor {colorUseRule=UseColor_Specified, specifiedColor=<rgba 1 1 1 1>}
      uVRegion {bIsValid=false, max=<0 0>, min=<0 0>}
     }
     horizontalScrolledContentPadding=<ltrb 0 0 1 0>
     leftShadowBrush {
      bIsDynamicallyLoaded=false
      drawAs=NoDrawType
      imageType=FullColor
      mirroring=NoMirror
      resourceName=../../../Engine/Content/Slate/Common/ScrollBoxShadowBottom.png
      resourceObject=null
      tiling=NoTile
      imageSize=<16 8>
      margin=<ltrb 0.5 0 0.5 1>
      outlineSettings {
       bUseBrushTransparency=false
       roundingType=HalfHeightRadius
       width=0
       color {colorUseRule=UseColor_Specified, specifiedColor=<rgba 0 0 0 0>}
       cornerRadii=<0 0 0 0>
      }
      tintColor {colorUseRule=UseColor_Specified, specifiedColor=<rgba 1 1 1 1>}
      uVRegion {bIsValid=false, max=<0 0>, min=<0 0>}
     }
     rightShadowBrush {
      bIsDynamicallyLoaded=false
      drawAs=NoDrawType
      imageType=FullColor
      mirroring=NoMirror
      resourceName=../../../Engine/Content/Slate/Common/ScrollBoxShadowBottom.png
      resourceObject=null
      tiling=NoTile
      imageSize=<16 8>
      margin=<ltrb 0.5 0 0.5 1>
      outlineSettings {
       bUseBrushTransparency=false
       roundingType=HalfHeightRadius
       width=0
       color {colorUseRule=UseColor_Specified, specifiedColor=<rgba 0 0 0 0>}
       cornerRadii=<0 0 0 0>
      }
      tintColor {colorUseRule=UseColor_Specified, specifiedColor=<rgba 1 1 1 1>}
      uVRegion {bIsValid=false, max=<0 0>, min=<0 0>}
     }
     topShadowBrush {
      bIsDynamicallyLoaded=false
      drawAs=NoDrawType
      imageType=FullColor
      mirroring=NoMirror
      resourceName=../../../Engine/Content/Slate/Common/ScrollBoxShadowTop.png
      resourceObject=null
      tiling=NoTile
      imageSize=<16 8>
      margin=<ltrb 0.5 1 0.5 0>
      outlineSettings {
       bUseBrushTransparency=false
       roundingType=HalfHeightRadius
       width=0
       color {colorUseRule=UseColor_Specified, specifiedColor=<rgba 0 0 0 0>}
       cornerRadii=<0 0 0 0>
      }
      tintColor {colorUseRule=UseColor_Specified, specifiedColor=<rgba 1 1 1 0.25>}
      uVRegion {bIsValid=false, max=<0 0>, min=<0 0>}
     }
     verticalScrolledContentPadding=<ltrb 0 0 0 1>
    }
   }
   slot {class=OverlaySlot, properties {horizontalAlignment=HAlign_Fill, verticalAlignment=VAlign_Fill, padding=<ltrb 100 220 80 180>}}
  }
  VerticalBox_1 {widgetClassPath=&VerticalBox, parent=CommonHierarchicalScrollBox_73, widget=VerticalBox_1, slot {class=ScrollBoxSlot}}
  VerticalBox_Presets {
   widgetClassPath=&VerticalBox
   parent=VerticalBox_1
   widget=VerticalBox_Presets
   slot {class=VerticalBoxSlot, properties {padding=<ltrb 0 0 0 30>}}
  }
  MainLayoutsTitle {
   widgetClassPath=&CommonTextBlock
   parent=VerticalBox_Presets
   widget=MainLayoutsTitle
   properties {
    text=Presets
    textTransformPolicy=ToUpper
    colorAndOpacity {colorUseRule=UseColor_Specified, specifiedColor=<rgba 0.1529 0.3515 0.807 1>}
    font {
     bForceMonospaced=false
     bMaterialIsStencil=false
     fontMaterial=null
     fontObject=&BurbankSmall
     letterSpacing=0
     monospacedWidth=1
     size=30
     skewAmount=0
     typefaceFontName=Black
     outlineSettings {
      bApplyOutlineToDropShadows=false
      bMiteredCorners=false
      bSeparateFillAlpha=false
      outlineMaterial=null
      outlineSize=0
      outlineColor=<rgba 0 0 0 1>
     }
    }
    shadowOffset=<0 0>
   }
   slot {class=VerticalBoxSlot, properties {horizontalAlignment=HAlign_Left, padding=<ltrb 0 0 0 10>}}
  }
  EntryBox_Presets {
   widgetClassPath=/Script/FortniteUI.FortDynamicEntryBox
   bInherited=true
   bIsVariable=true
   parent=VerticalBox_Presets
   widget=EntryBox_Presets
   properties {
    entryBoxType=Wrap
    entryWidgetClass=&/HUDLayoutToolV2_LayoutButton.HUDLayoutToolV2_LayoutButton_C
    entrySpacing=<40 40>
   }
   slot {class=VerticalBoxSlot, properties {padding=<ltrb 8 8 8 8>, size {sizeRule=Fill, value=1}}}
  }
  VerticalBox_Customs {
   widgetClassPath=&VerticalBox
   parent=VerticalBox_1
   widget=VerticalBox_Customs
   slot {class=VerticalBoxSlot, properties {padding=<ltrb 0 30 0 0>}}
  }
  MainLayoutsTitle_1 {
   widgetClassPath=&CommonTextBlock
   parent=VerticalBox_Customs
   widget=MainLayoutsTitle_1
   properties {
    text=Custom Layouts
    textTransformPolicy=ToUpper
    colorAndOpacity {colorUseRule=UseColor_Specified, specifiedColor=<rgba 0.1529 0.3515 0.807 1>}
    font {
     bForceMonospaced=false
     bMaterialIsStencil=false
     fontMaterial=null
     fontObject=&BurbankSmall
     letterSpacing=0
     monospacedWidth=1
     size=30
     skewAmount=0
     typefaceFontName=Black
     outlineSettings {
      bApplyOutlineToDropShadows=false
      bMiteredCorners=false
      bSeparateFillAlpha=false
      outlineMaterial=null
      outlineSize=0
      outlineColor=<rgba 0 0 0 1>
     }
    }
    shadowOffset=<0 0>
   }
   slot {class=VerticalBoxSlot, properties {horizontalAlignment=HAlign_Left, padding=<ltrb 0 0 0 10>}}
  }
  Overlay_Customs {widgetClassPath=&Overlay, parent=VerticalBox_Customs, widget=Overlay_Customs, slot {class=VerticalBoxSlot}}
  EntryBox_EmptyCustomLayoutSlots {
   widgetClassPath=&FortDynamicEntryBox
   bInherited=true
   bIsVariable=true
   parent=Overlay_Customs
   widget=EntryBox_EmptyCustomLayoutSlots
   properties {
    entryBoxType=Wrap
    entryWidgetClass=&/HUDLayoutToolV2_EmptyCustomLayoutSlot.HUDLayoutToolV2_EmptyCustomLayoutSlot_C
    numDesignerPreviewEntries=6
    entrySpacing=<40 40>
   }
   slot {class=OverlaySlot, properties {horizontalAlignment=HAlign_Fill, verticalAlignment=VAlign_Fill, padding=<ltrb 8 8 8 8>}}
  }
  EntryBox_CustomLayouts {
   widgetClassPath=&FortDynamicEntryBox
   bInherited=true
   bIsVariable=true
   parent=Overlay_Customs
   widget=EntryBox_CustomLayouts
   properties {
    entryBoxType=Wrap
    entryWidgetClass=&HUDLayoutToolV2_LayoutButton_C
    entrySpacing=<40 40>
   }
   slot {class=OverlaySlot, properties {horizontalAlignment=HAlign_Fill, verticalAlignment=VAlign_Fill, padding=<ltrb 8 8 8 8>}}
  }
  Overlay_252 {
   widgetClassPath=&Overlay
   parent=Overlay_177
   widget=Overlay_252
   slot {class=OverlaySlot, properties {horizontalAlignment=HAlign_Fill, verticalAlignment=VAlign_Bottom, padding=<ltrb 100 0 100 50>}}
  }
  HorizontalBox_248 {
   widgetClassPath=&HorizontalBox
   parent=Overlay_252
   widget=HorizontalBox_248
   slot {class=OverlaySlot, properties {verticalAlignment=VAlign_Center}}
  }
  Button_Import {
   widgetClassPath=/Game/Athena/HUD/Bacchus/HUDLayoutTool/HUDLayoutToolV2_Button
   parent=HorizontalBox_248
   widget=Button_Import
   properties {
    bOverride_HeightOverride=true
    bOverride_WidthOverride=true
    buttonHeightOverride=80
    buttonWidthOverride=246
    text=Import
    textFontSize=26
    visibility=Collapsed
    normalButtonStyle {
     iconAngle=0
     backgroundColor=<rgba 0.0056 0.0222 0.3005 1>
     iconColor=<rgba 0.0319 0.7835 0.9473 1>
     textColor {colorUseRule=UseColor_Specified, specifiedColor=<rgba 0.0319 0.7835 0.9473 1>}
    }
    pressedButtonStyle {
     iconAngle=0
     backgroundColor=<rgba 0 0.1651 0.9911 1>
     iconColor=<rgba 1 1 1 1>
     textColor {colorUseRule=UseColor_Specified, specifiedColor=<rgba 1 1 1 1>}
    }
    selectedButtonStyle {
     iconAngle=0
     backgroundColor=<rgba 0 0.1651 1 1>
     iconColor=<rgba 1 1 1 1>
     textColor {colorUseRule=UseColor_Specified, specifiedColor=<rgba 1 1 1 1>}
    }
   }
   slot {class=HorizontalBoxSlot, properties {padding=<ltrb 0 0 20 0>}}
  }
  Button_Share {
   widgetClassPath=&HUDLayoutToolV2_Button
   parent=HorizontalBox_248
   widget=Button_Share
   properties {
    bOverride_HeightOverride=true
    bOverride_WidthOverride=true
    buttonHeightOverride=80
    buttonWidthOverride=246
    text=Share
    textFontSize=26
    visibility=Collapsed
    normalButtonStyle {
     iconAngle=0
     backgroundColor=<rgba 0.0056 0.0222 0.3005 1>
     iconColor=<rgba 0.0319 0.7835 0.9473 1>
     textColor {colorUseRule=UseColor_Specified, specifiedColor=<rgba 0.0319 0.7835 0.9473 1>}
    }
    pressedButtonStyle {
     iconAngle=0
     backgroundColor=<rgba 0 0.1651 1 1>
     iconColor=<rgba 1 1 1 1>
     textColor {colorUseRule=UseColor_Specified, specifiedColor=<rgba 1 1 1 1>}
    }
    selectedButtonStyle {
     iconAngle=0
     backgroundColor=<rgba 0 0.1651 1 1>
     iconColor=<rgba 1 1 1 1>
     textColor {colorUseRule=UseColor_Specified, specifiedColor=<rgba 1 1 1 1>}
    }
   }
   slot {class=HorizontalBoxSlot, properties {padding=<ltrb 0 0 20 0>}}
  }
  Button_Delete {
   widgetClassPath=&HUDLayoutToolV2_Button
   bInherited=true
   parent=HorizontalBox_248
   widget=Button_Delete
   properties {
    bOverride_HeightOverride=true
    bOverride_WidthOverride=true
    buttonHeightOverride=80
    buttonWidthOverride=288
    text=Delete
    textFontSize=26
    internalPadding=<ltrb 24 12 24 12>
    normalButtonStyle {
     iconAngle=0
     backgroundColor=<rgba 0.1714 0.0033 0.0033 1>
     iconColor=<rgba 1 0.1714 0.1329 1>
     textColor {colorUseRule=UseColor_Specified, specifiedColor=<rgba 1 0.1714 0.1329 1>}
    }
    pressedButtonStyle {
     iconAngle=0
     backgroundColor=<rgba 0.6939 0.0137 0.016 1>
     iconColor=<rgba 1 1 1 1>
     textColor {colorUseRule=UseColor_Specified, specifiedColor=<rgba 1 1 1 1>}
    }
    selectedButtonStyle {
     iconAngle=0
     backgroundColor=<rgba 0.6939 0.0137 0.016 1>
     iconColor=<rgba 1 1 1 1>
     textColor {colorUseRule=UseColor_Specified, specifiedColor=<rgba 1 1 1 1>}
    }
   }
   slot {class=HorizontalBoxSlot, properties {padding=<ltrb 0 0 20 0>}}
  }
  Button_Rename {
   widgetClassPath=&HUDLayoutToolV2_Button
   bInherited=true
   parent=HorizontalBox_248
   widget=Button_Rename
   properties {
    bOverride_HeightOverride=true
    bOverride_WidthOverride=true
    buttonHeightOverride=80
    buttonWidthOverride=288
    text=Rename
    textFontSize=26
    internalPadding=<ltrb 24 12 24 12>
    normalButtonStyle {
     iconAngle=0
     backgroundColor=<rgba 0.0056 0.0222 0.3005 1>
     iconColor=<rgba 0.0319 0.7835 0.9473 1>
     textColor {colorUseRule=UseColor_Specified, specifiedColor=<rgba 0.0319 0.7835 0.9473 1>}
    }
    pressedButtonStyle {
     iconAngle=0
     backgroundColor=<rgba 0 0.1647 0.9922 1>
     iconColor=<rgba 1 1 1 1>
     textColor {colorUseRule=UseColor_Specified, specifiedColor=<rgba 1 1 1 1>}
    }
    selectedButtonStyle {
     iconAngle=0
     backgroundColor=<rgba 0 0.1651 1 1>
     iconColor=<rgba 1 1 1 1>
     textColor {colorUseRule=UseColor_Specified, specifiedColor=<rgba 1 1 1 1>}
    }
   }
   slot {class=HorizontalBoxSlot, properties {padding=<ltrb 0 0 20 0>}}
  }
  Button_Copy {
   widgetClassPath=&HUDLayoutToolV2_Button
   bInherited=true
   parent=HorizontalBox_248
   widget=Button_Copy
   properties {
    bOverride_HeightOverride=true
    bOverride_WidthOverride=true
    buttonHeightOverride=80
    buttonWidthOverride=320
    text=Create new
    textFontSize=26
    internalPadding=<ltrb 24 12 24 12>
    normalButtonStyle {
     iconAngle=0
     backgroundColor=<rgba 0.0056 0.0222 0.3005 1>
     iconColor=<rgba 0.0319 0.7835 0.9473 1>
     textColor {colorUseRule=UseColor_Specified, specifiedColor=<rgba 0.0319 0.7835 0.9473 1>}
    }
    pressedButtonStyle {
     iconAngle=0
     backgroundColor=<rgba 0 0.1647 0.9922 1>
     iconColor=<rgba 1 1 1 1>
     textColor {colorUseRule=UseColor_Specified, specifiedColor=<rgba 1 1 1 1>}
    }
    selectedButtonStyle {
     iconAngle=0
     backgroundColor=<rgba 0 0.1651 1 1>
     iconColor=<rgba 1 1 1 1>
     textColor {colorUseRule=UseColor_Specified, specifiedColor=<rgba 1 1 1 1>}
    }
   }
   slot {class=HorizontalBoxSlot, properties {padding=<ltrb 0 0 20 0>}}
  }
  Button_Convert {
   widgetClassPath=&HUDLayoutToolV2_Button
   bInherited=true
   parent=HorizontalBox_248
   widget=Button_Convert
   properties {
    bOverride_HeightOverride=true
    bOverride_WidthOverride=true
    buttonHeightOverride=80
    buttonWidthOverride=320
    text=Convert Legacy HUD
    textFontSize=26
    internalPadding=<ltrb 24 12 24 12>
    normalButtonStyle {
     iconAngle=0
     backgroundColor=<rgba 0.0056 0.0222 0.3005 1>
     iconColor=<rgba 0.0319 0.7835 0.9473 1>
     textColor {colorUseRule=UseColor_Specified, specifiedColor=<rgba 0.0319 0.7835 0.9473 1>}
    }
    pressedButtonStyle {
     iconAngle=0
     backgroundColor=<rgba 0 0.1647 0.9922 1>
     iconColor=<rgba 1 1 1 1>
     textColor {colorUseRule=UseColor_Specified, specifiedColor=<rgba 1 1 1 1>}
    }
    selectedButtonStyle {
     iconAngle=0
     backgroundColor=<rgba 0 0.1651 1 1>
     iconColor=<rgba 1 1 1 1>
     textColor {colorUseRule=UseColor_Specified, specifiedColor=<rgba 1 1 1 1>}
    }
   }
   slot {class=HorizontalBoxSlot, properties {padding=<ltrb 0 0 20 0>}}
  }
  Button_EditLayout {
   widgetClassPath=&HUDLayoutToolV2_Button
   bInherited=true
   parent=Overlay_252
   widget=Button_EditLayout
   properties {
    bOverride_HeightOverride=true
    bOverride_WidthOverride=true
    buttonHeightOverride=80
    buttonWidthOverride=390
    text=Select layout
    textFontSize=26
    internalPadding=<ltrb 24 12 24 12>
    normalButtonStyle {
     iconAngle=0
     backgroundColor=<rgba 0.9301 1 0.0097 1>
     iconColor=<rgba 0 0 0 1>
     textColor {colorUseRule=UseColor_Specified, specifiedColor=<rgba 0 0 0 1>}
    }
    pressedButtonStyle {
     iconAngle=0
     backgroundColor=<rgba 1 1 1 1>
     iconColor=<rgba 0 0 0 1>
     textColor {colorUseRule=UseColor_Specified, specifiedColor=<rgba 0 0 0 1>}
    }
   }
   slot {class=OverlaySlot, properties {horizontalAlignment=HAlign_Right, verticalAlignment=VAlign_Center}}
  }
 ]
}
