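# Toast notification widget for the HUD Layout Tool, showing a success or failure result on a sliding, sheared panel.
# The event graph is empty; the native HUDLayoutToolV2_Toaster parent supplies the inherited title and description text blocks.
# Success shows a green status bar and popping checkmark while collapsing the title.
# Failure, LocalFailure and CloudFailure animations have identical tracks, showing a yellow bar, title and spinning warning icon.
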WidgetBlueprint /FortniteCoreUI/Athena/HUD/Bacchus/HUDLayoutTool/HUDLayoutToolV2_Toaster
animations [
 Anim_Success {
  binding_count=5
  length_seconds=2.5362
  object=&self:Anim_Success
  sequence {bindings [
    CommonText_Title {
     id=#1
     possessed_class=CommonTextBlock
     tracks [{
       type=MovieSceneByteTrack
       display_name=Visibility
       sections [{
         type=MovieSceneByteSection
         end_frame=51
         start_frame=0
         channels [None {type=MovieSceneScriptingByteChannel, keys [[0.0073,4],[0.457,1],[49.1833,1],[50.011,4]]}]
        }]
      }]
    }
    Image_Checkmark {
     id=#2
     possessed_class=Image
     tracks [
      {
       type=MovieSceneFloatTrack
       display_name=Render Opacity
       sections [{
         type=MovieSceneFloatSection
         end_frame=51
         start_frame=0
         channels [None {type=MovieSceneScriptingFloatChannel, thinned_from=4, keys [[0.0073,0],[10.399,0],[13.774,1]]}]
        }]
      }
      {
       type=MovieScene2DTransformTrack
       sections [{
         type=MovieScene2DTransformSection
         end_frame=51
         start_frame=-1
         channels [
          Translation.X {type=MovieSceneScriptingFloatChannel}
          Translation.Y {type=MovieSceneScriptingFloatChannel, keys [[8.4063,64],[12.6333,0]]}
          Angle {type=MovieSceneScriptingFloatChannel, keys [[8.4063,360],[14.397,0]]}
          Scale.X {type=MovieSceneScriptingFloatChannel, keys [[8.4063,4],[14.397,0.75],[17.074,1.25],[20.0857,1]]}
          Scale.Y {type=MovieSceneScriptingFloatChannel, keys [[8.4063,4],[14.397,0.75],[16.9747,1.25],[20.0857,1]]}
          Shear.X {type=MovieSceneScriptingFloatChannel}
          Shear.Y {type=MovieSceneScriptingFloatChannel}
         ]
        }]
      }
     ]
    }
    SizeBox_16 {
     id=#3
     possessed_class=SizeBox
     tracks [
      {
       type=MovieScene2DTransformTrack
       sections [{
         type=MovieScene2DTransformSection
         end_frame=50
         start_frame=-1
         channels [
          Translation.X {type=MovieSceneScriptingFloatChannel, keys [[0.0073,-800],[8.4063,0],[44.6793,0],[49.914,-800]]}
          Translation.Y {type=MovieSceneScriptingFloatChannel}
          Angle {type=MovieSceneScriptingFloatChannel}
          Scale.X {type=MovieSceneScriptingFloatChannel}
          Scale.Y {type=MovieSceneScriptingFloatChannel}
          Shear.X {type=MovieSceneScriptingFloatChannel, keys [[0.0073,16],[5.4807,11.25],[8.4063,0],[44.6793,0],[48.0663,11.5674],[49.914,16]]}
          Shear.Y {type=MovieSceneScriptingFloatChannel}
         ]
        }]
      }
      {
       type=MovieSceneFloatTrack
       display_name=Render Opacity
       sections [{
         type=MovieSceneFloatSection
         end_frame=51
         start_frame=-1
         channels [None {type=MovieSceneScriptingFloatChannel, keys [[0.0073,0],[8.4063,1],[44.6793,1],[49.9137,0]]}]
        }]
      }
      {
       type=MovieSceneByteTrack
       display_name=Visibility
       sections [{
         type=MovieSceneByteSection
         end_frame=51
         start_frame=0
         channels [None {type=MovieSceneScriptingByteChannel, keys [[0.0073,0],[49.914,4],[50.0303,1]]}]
        }]
      }
     ]
    }
    SizeBox_81 {
     id=#4
     possessed_class=SizeBox
     tracks [{
       type=MovieSceneByteTrack
       display_name=Visibility
       sections [{type=MovieSceneByteSection, end_frame=51, start_frame=0, channels [None {type=MovieSceneScriptingByteChannel, const [1,2]}]}]
      }]
    }
    Image_StatusIndicator {
     id=#5
     possessed_class=Image
     tracks [{
       type=MovieSceneColorTrack
       display_name=Color and Opacity
       sections [{
         type=MovieSceneColorSection
         end_frame=51
         start_frame=0
         channels [
          Color.R {type=MovieSceneScriptingFloatChannel, keys [[0.0073,1],[3.3513,0.0343],[46.3517,0.0343],[49.2147,1]]}
          Color.G {type=MovieSceneScriptingFloatChannel, const [1,4]}
          Color.B {type=MovieSceneScriptingFloatChannel, keys [[0.0073,1],[3.3513,0.227],[46.3517,0.227],[49.2147,1]]}
          Color.A {type=MovieSceneScriptingFloatChannel, const [1,4]}
         ]
        }]
      }]
    }
   ]}
 }
 Anim_Failure {
  binding_count=6
  length_seconds=2.5362
  object=&self:Anim_Failure
  sequence {bindings [
    Image_Warning {
     id=#6
     possessed_class=Image
     tracks [
      {
       type=MovieSceneFloatTrack
       display_name=Render Opacity
       sections [{
         type=MovieSceneFloatSection
         end_frame=51
         start_frame=-1
         channels [None {type=MovieSceneScriptingFloatChannel, thinned_from=4, keys [[-0.0033,0],[10.3883,0],[13.7633,1]]}]
        }]
      }
      {
       type=MovieScene2DTransformTrack
       sections [{
         type=MovieScene2DTransformSection
         end_frame=51
         start_frame=-1
         channels [
          Translation.X {type=MovieSceneScriptingFloatChannel, const [0,1]}
          Translation.Y {type=MovieSceneScriptingFloatChannel, keys [[8.4053,64],[12.6323,0]]}
          Angle {type=MovieSceneScriptingFloatChannel, keys [[8.4053,360],[14.396,0]]}
          Scale.X {type=MovieSceneScriptingFloatChannel, keys [[8.4053,4],[14.396,0.75],[17.073,1.25],[20.0847,1]]}
          Scale.Y {type=MovieSceneScriptingFloatChannel, keys [[8.4053,4],[14.396,0.75],[16.9737,1.25],[20.0847,1]]}
          Shear.X {type=MovieSceneScriptingFloatChannel, const [0,1]}
          Shear.Y {type=MovieSceneScriptingFloatChannel, const [0,1]}
         ]
        }]
      }
     ]
    }
    CommonText_Title {
     id=#1
     possessed_class=CommonTextBlock
     tracks [{
       type=MovieSceneByteTrack
       display_name=Visibility
       sections [{type=MovieSceneByteSection, end_frame=51, start_frame=0, channels [None {type=MovieSceneScriptingByteChannel, const [4,3]}]}]
      }]
    }
    Image_Checkmark {
     id=#2
     possessed_class=Image
     tracks [{
       type=MovieSceneByteTrack
       display_name=Visibility
       sections [{
         type=MovieSceneByteSection
         end_frame=51
         start_frame=0
         channels [None {type=MovieSceneScriptingByteChannel, keys [[0,0],[2,1],[48.9807,1],[50.0323,0]]}]
        }]
      }]
    }
    SizeBox_16 {
     id=#3
     possessed_class=SizeBox
     tracks [
      {
       type=MovieScene2DTransformTrack
       sections [{
         type=MovieScene2DTransformSection
         end_frame=50
         start_frame=-1
         channels [
          Translation.X {type=MovieSceneScriptingFloatChannel, keys [[0.0073,-800],[8.4063,0],[44.6793,0],[49.914,-800]]}
          Translation.Y {type=MovieSceneScriptingFloatChannel}
          Angle {type=MovieSceneScriptingFloatChannel}
          Scale.X {type=MovieSceneScriptingFloatChannel}
          Scale.Y {type=MovieSceneScriptingFloatChannel}
          Shear.X {type=MovieSceneScriptingFloatChannel, keys [[0.0073,16],[5.4807,11.25],[8.4063,0],[44.6793,0],[48.0663,11.5674],[49.914,16]]}
          Shear.Y {type=MovieSceneScriptingFloatChannel}
         ]
        }]
      }
      {
       type=MovieSceneFloatTrack
       display_name=Render Opacity
       sections [{
         type=MovieSceneFloatSection
         end_frame=51
         start_frame=-1
         channels [None {type=MovieSceneScriptingFloatChannel, keys [[0.0073,0],[8.4063,1],[44.6793,1],[49.9137,0]]}]
        }]
      }
      {
       type=MovieSceneByteTrack
       display_name=Visibility
       sections [{
         type=MovieSceneByteSection
         end_frame=51
         start_frame=0
         channels [None {type=MovieSceneScriptingByteChannel, keys [[0.0073,0],[49.914,4],[50.0303,1]]}]
        }]
      }
     ]
    }
    SizeBox_81 {
     id=#4
     possessed_class=SizeBox
     tracks [{
       type=MovieSceneByteTrack
       display_name=Visibility
       sections [{
         type=MovieSceneByteSection
         end_frame=51
         start_frame=0
         channels [None {type=MovieSceneScriptingByteChannel, keys [[0.0073,1],[0.8977,4],[49.565,4],[50.091,1]]}]
        }]
      }]
    }
    Image_StatusIndicator {
     id=#5
     possessed_class=Image
     tracks [{
       type=MovieSceneColorTrack
       display_name=Color and Opacity
       sections [{
         type=MovieSceneColorSection
         end_frame=51
         start_frame=0
         channels [
          Color.R {type=MovieSceneScriptingFloatChannel, keys [[0.0073,1],[3.3513,0.9301],[46.3517,0.9301],[49.2147,1]]}
          Color.G {type=MovieSceneScriptingFloatChannel, const [1,4]}
          Color.B {type=MovieSceneScriptingFloatChannel, keys [[0.0073,1],[3.3513,0.0097],[46.3517,0.0097],[49.2147,1]]}
          Color.A {type=MovieSceneScriptingFloatChannel, const [1,4]}
         ]
        }]
      }]
    }
   ]}
 }
 Anim_LocalFailure {
  binding_count=6
  length_seconds=2.5362
  object=&self:Anim_LocalFailure
  sequence {bindings [
    Image_Warning {
     id=#6
     possessed_class=Image
     tracks [
      {
       type=MovieSceneFloatTrack
       display_name=Render Opacity
       sections [{
         type=MovieSceneFloatSection
         end_frame=51
         start_frame=-1
         channels [None {type=MovieSceneScriptingFloatChannel, thinned_from=4, keys [[-0.0033,0],[10.3883,0],[13.7633,1]]}]
        }]
      }
      {
       type=MovieScene2DTransformTrack
       sections [{
         type=MovieScene2DTransformSection
         end_frame=51
         start_frame=-1
         channels [
          Translation.X {type=MovieSceneScriptingFloatChannel, const [0,1]}
          Translation.Y {type=MovieSceneScriptingFloatChannel, keys [[8.4053,64],[12.6323,0]]}
          Angle {type=MovieSceneScriptingFloatChannel, keys [[8.4053,360],[14.396,0]]}
          Scale.X {type=MovieSceneScriptingFloatChannel, keys [[8.4053,4],[14.396,0.75],[17.073,1.25],[20.0847,1]]}
          Scale.Y {type=MovieSceneScriptingFloatChannel, keys [[8.4053,4],[14.396,0.75],[16.9737,1.25],[20.0847,1]]}
          Shear.X {type=MovieSceneScriptingFloatChannel, const [0,1]}
          Shear.Y {type=MovieSceneScriptingFloatChannel, const [0,1]}
         ]
        }]
      }
     ]
    }
    CommonText_Title {
     id=#1
     possessed_class=CommonTextBlock
     tracks [{
       type=MovieSceneByteTrack
       display_name=Visibility
       sections [{type=MovieSceneByteSection, end_frame=51, start_frame=0, channels [None {type=MovieSceneScriptingByteChannel, const [4,3]}]}]
      }]
    }
    Image_Checkmark {
     id=#2
     possessed_class=Image
     tracks [{
       type=MovieSceneByteTrack
       display_name=Visibility
       sections [{
         type=MovieSceneByteSection
         end_frame=51
         start_frame=0
         channels [None {type=MovieSceneScriptingByteChannel, keys [[0,0],[2,1],[48.9807,1],[50.0323,0]]}]
        }]
      }]
    }
    SizeBox_16 {
     id=#3
     possessed_class=SizeBox
     tracks [
      {
       type=MovieScene2DTransformTrack
       sections [{
         type=MovieScene2DTransformSection
         end_frame=50
         start_frame=-1
         channels [
          Translation.X {type=MovieSceneScriptingFloatChannel, keys [[0.0073,-800],[8.4063,0],[44.6793,0],[49.914,-800]]}
          Translation.Y {type=MovieSceneScriptingFloatChannel}
          Angle {type=MovieSceneScriptingFloatChannel}
          Scale.X {type=MovieSceneScriptingFloatChannel}
          Scale.Y {type=MovieSceneScriptingFloatChannel}
          Shear.X {type=MovieSceneScriptingFloatChannel, keys [[0.0073,16],[5.4807,11.25],[8.4063,0],[44.6793,0],[48.0663,11.5674],[49.914,16]]}
          Shear.Y {type=MovieSceneScriptingFloatChannel}
         ]
        }]
      }
      {
       type=MovieSceneFloatTrack
       display_name=Render Opacity
       sections [{
         type=MovieSceneFloatSection
         end_frame=51
         start_frame=-1
         channels [None {type=MovieSceneScriptingFloatChannel, keys [[0.0073,0],[8.4063,1],[44.6793,1],[49.9137,0]]}]
        }]
      }
      {
       type=MovieSceneByteTrack
       display_name=Visibility
       sections [{
         type=MovieSceneByteSection
         end_frame=51
         start_frame=0
         channels [None {type=MovieSceneScriptingByteChannel, keys [[0.0073,0],[49.914,4],[50.0303,1]]}]
        }]
      }
     ]
    }
    SizeBox_81 {
     id=#4
     possessed_class=SizeBox
     tracks [{
       type=MovieSceneByteTrack
       display_name=Visibility
       sections [{
         type=MovieSceneByteSection
         end_frame=51
         start_frame=0
         channels [None {type=MovieSceneScriptingByteChannel, keys [[0.0073,1],[0.8977,4],[49.565,4],[50.091,1]]}]
        }]
      }]
    }
    Image_StatusIndicator {
     id=#5
     possessed_class=Image
     tracks [{
       type=MovieSceneColorTrack
       display_name=Color and Opacity
       sections [{
         type=MovieSceneColorSection
         end_frame=51
         start_frame=0
         channels [
          Color.R {type=MovieSceneScriptingFloatChannel, keys [[0.0073,1],[3.3513,0.9301],[46.3517,0.9301],[49.2147,1]]}
          Color.G {type=MovieSceneScriptingFloatChannel, const [1,4]}
          Color.B {type=MovieSceneScriptingFloatChannel, keys [[0.0073,1],[3.3513,0.0097],[46.3517,0.0097],[49.2147,1]]}
          Color.A {type=MovieSceneScriptingFloatChannel, const [1,4]}
         ]
        }]
      }]
    }
   ]}
 }
 Anim_CloudFailure {
  binding_count=6
  length_seconds=2.5362
  object=&self:Anim_CloudFailure
  sequence {bindings [
    Image_Warning {
     id=#6
     possessed_class=Image
     tracks [
      {
       type=MovieSceneFloatTrack
       display_name=Render Opacity
       sections [{
         type=MovieSceneFloatSection
         end_frame=51
         start_frame=-1
         channels [None {type=MovieSceneScriptingFloatChannel, thinned_from=4, keys [[-0.0033,0],[10.3883,0],[13.7633,1]]}]
        }]
      }
      {
       type=MovieScene2DTransformTrack
       sections [{
         type=MovieScene2DTransformSection
         end_frame=51
         start_frame=-1
         channels [
          Translation.X {type=MovieSceneScriptingFloatChannel, const [0,1]}
          Translation.Y {type=MovieSceneScriptingFloatChannel, keys [[8.4053,64],[12.6323,0]]}
          Angle {type=MovieSceneScriptingFloatChannel, keys [[8.4053,360],[14.396,0]]}
          Scale.X {type=MovieSceneScriptingFloatChannel, keys [[8.4053,4],[14.396,0.75],[17.073,1.25],[20.0847,1]]}
          Scale.Y {type=MovieSceneScriptingFloatChannel, keys [[8.4053,4],[14.396,0.75],[16.9737,1.25],[20.0847,1]]}
          Shear.X {type=MovieSceneScriptingFloatChannel, const [0,1]}
          Shear.Y {type=MovieSceneScriptingFloatChannel, const [0,1]}
         ]
        }]
      }
     ]
    }
    CommonText_Title {
     id=#1
     possessed_class=CommonTextBlock
     tracks [{
       type=MovieSceneByteTrack
       display_name=Visibility
       sections [{type=MovieSceneByteSection, end_frame=51, start_frame=0, channels [None {type=MovieSceneScriptingByteChannel, const [4,3]}]}]
      }]
    }
    Image_Checkmark {
     id=#2
     possessed_class=Image
     tracks [{
       type=MovieSceneByteTrack
       display_name=Visibility
       sections [{
         type=MovieSceneByteSection
         end_frame=51
         start_frame=0
         channels [None {type=MovieSceneScriptingByteChannel, keys [[0,0],[2,1],[48.9807,1],[50.0323,0]]}]
        }]
      }]
    }
    SizeBox_16 {
     id=#3
     possessed_class=SizeBox
     tracks [
      {
       type=MovieScene2DTransformTrack
       sections [{
         type=MovieScene2DTransformSection
         end_frame=50
         start_frame=-1
         channels [
          Translation.X {type=MovieSceneScriptingFloatChannel, keys [[0.0073,-800],[8.4063,0],[44.6793,0],[49.914,-800]]}
          Translation.Y {type=MovieSceneScriptingFloatChannel}
          Angle {type=MovieSceneScriptingFloatChannel}
          Scale.X {type=MovieSceneScriptingFloatChannel}
          Scale.Y {type=MovieSceneScriptingFloatChannel}
          Shear.X {type=MovieSceneScriptingFloatChannel, keys [[0.0073,16],[5.4807,11.25],[8.4063,0],[44.6793,0],[48.0663,11.5674],[49.914,16]]}
          Shear.Y {type=MovieSceneScriptingFloatChannel}
         ]
        }]
      }
      {
       type=MovieSceneFloatTrack
       display_name=Render Opacity
       sections [{
         type=MovieSceneFloatSection
         end_frame=51
         start_frame=-1
         channels [None {type=MovieSceneScriptingFloatChannel, keys [[0.0073,0],[8.4063,1],[44.6793,1],[49.9137,0]]}]
        }]
      }
      {
       type=MovieSceneByteTrack
       display_name=Visibility
       sections [{
         type=MovieSceneByteSection
         end_frame=51
         start_frame=0
         channels [None {type=MovieSceneScriptingByteChannel, keys [[0.0073,0],[49.914,4],[50.0303,1]]}]
        }]
      }
     ]
    }
    SizeBox_81 {
     id=#4
     possessed_class=SizeBox
     tracks [{
       type=MovieSceneByteTrack
       display_name=Visibility
       sections [{
         type=MovieSceneByteSection
         end_frame=51
         start_frame=0
         channels [None {type=MovieSceneScriptingByteChannel, keys [[0.0073,1],[0.8977,4],[49.565,4],[50.091,1]]}]
        }]
      }]
    }
    Image_StatusIndicator {
     id=#5
     possessed_class=Image
     tracks [{
       type=MovieSceneColorTrack
       display_name=Color and Opacity
       sections [{
         type=MovieSceneColorSection
         end_frame=51
         start_frame=0
         channels [
          Color.R {type=MovieSceneScriptingFloatChannel, keys [[0.0073,1],[3.3513,0.9301],[46.3517,0.9301],[49.2147,1]]}
          Color.G {type=MovieSceneScriptingFloatChannel, const [1,4]}
          Color.B {type=MovieSceneScriptingFloatChannel, keys [[0.0073,1],[3.3513,0.0097],[46.3517,0.0097],[49.2147,1]]}
          Color.A {type=MovieSceneScriptingFloatChannel, const [1,4]}
         ]
        }]
      }]
    }
   ]}
 }
]
graphs [EventGraph {links [], nodes []}]
variables []
widgets {
 info {inheritedWidgetCount=2, parentClass=/Script/FortniteUI.HUDLayoutToolV2_Toaster, rootWidgetClass=&Overlay, widgetCount=15}
 widgets [
  Overlay_0 {widgetClassPath=&Overlay, widget=Overlay_0}
  SizeBox_16 {
   widgetClassPath=&SizeBox
   parent=Overlay_0
   widget=SizeBox_16
   properties {bOverride_MaxDesiredWidth=true, maxDesiredWidth=716, minDesiredHeight=127, minDesiredWidth=716}
   slot {class=OverlaySlot, properties {padding=<ltrb 0 100 0 0>}}
  }
  Overlay_53 {widgetClassPath=&Overlay, parent=SizeBox_16, widget=Overlay_53, slot {class=SizeBoxSlot, properties {padding=<ltrb 30 0 0 0>}}}
  Image_Shadow {
   widgetClassPath=&Image
   parent=Overlay_53
   widget=Image_Shadow
   properties {
    visibility=SelfHitTestInvisible
    brush {
     bIsDynamicallyLoaded=false
     drawAs=Image
     imageType=NoImage
     mirroring=NoMirror
     resourceName=None
     resourceObject=null
     tiling=NoTile
     imageSize=<32 32>
     margin=<ltrb 0 0 0 0>
     outlineSettings {
      bUseBrushTransparency=false
      roundingType=HalfHeightRadius
      width=0
      color {colorUseRule=UseColor_Specified, specifiedColor=<rgba 0 0 0 0>}
      cornerRadii=<0 0 0 0>
     }
     tintColor {colorUseRule=UseColor_Specified, specifiedColor=<rgba 0 0 0 0.2>}
     uVRegion {bIsValid=false, max=<0 0>, min=<0 0>}
    }
    renderTransform {angle=0, scale=<1 1>, shear=<0 0>, translation=<14 14>}
   }
   slot {class=OverlaySlot, properties {horizontalAlignment=HAlign_Fill, verticalAlignment=VAlign_Fill}}
  }
  Border_60 {
   widgetClassPath=&Border
   parent=Overlay_53
   widget=Border_60
   properties {
    background {
     bIsDynamicallyLoaded=false
     drawAs=Image
     imageType=NoImage
     mirroring=NoMirror
     resourceName=None
     resourceObject=/Game/UI/SocialPanel/Art/Materials/M_UI_SocialUserItemBorder_UserListItem
     tiling=NoTile
     imageSize=<32 32>
     margin=<ltrb 0 0 0 0>
     outlineSettings {
      bUseBrushTransparency=false
      roundingType=HalfHeightRadius
      width=0
      color {colorUseRule=UseColor_Specified, specifiedColor=<rgba 0 0 0 0>}
      cornerRadii=<0 0 0 0>
     }
     tintColor {colorUseRule=UseColor_Specified, specifiedColor=<rgba 1 1 1 1>}
     uVRegion {bIsValid=false, max=<0 0>, min=<0 0>}
    }
    padding=<ltrb 0 0 0 0>
   }
   slot {class=OverlaySlot, properties {horizontalAlignment=HAlign_Fill, verticalAlignment=VAlign_Fill}}
  }
  HorizontalBox_0 {
   widgetClassPath=&HorizontalBox
   parent=Border_60
   widget=HorizontalBox_0
   slot {class=BorderSlot, properties {padding=<ltrb 0 0 0 0>}}
  }
  Image_StatusIndicator {
   widgetClassPath=&Image
   parent=HorizontalBox_0
   widget=Image_StatusIndicator
   properties {
    brush {
     bIsDynamicallyLoaded=false
     drawAs=Image
     imageType=NoImage
     mirroring=NoMirror
     resourceName=None
     resourceObject=null
     tiling=NoTile
     imageSize=<16 32>
     margin=<ltrb 0 0 0 0>
     outlineSettings {
      bUseBrushTransparency=false
      roundingType=HalfHeightRadius
      width=0
      color {colorUseRule=UseColor_Specified, specifiedColor=<rgba 0 0 0 0>}
      cornerRadii=<0 0 0 0>
     }
     tintColor {colorUseRule=UseColor_Specified, specifiedColor=<rgba 1 1 1 1>}
     uVRegion {bIsValid=false, max=<0 0>, min=<0 0>}
    }
    colorAndOpacity=<rgba 0.0343 1 0.227 1>
   }
   slot {class=HorizontalBoxSlot}
  }
  VerticalBox_0 {
   widgetClassPath=&VerticalBox
   parent=HorizontalBox_0
   widget=VerticalBox_0
   slot {class=HorizontalBoxSlot, properties {horizontalAlignment=HAlign_Left, verticalAlignment=VAlign_Center, padding=<ltrb 30 20 0 20>}}
  }
  CommonText_Title {
   widgetClassPath=&CommonTextBlock
   bInherited=true
   parent=VerticalBox_0
   widget=CommonText_Title
   properties {
    text="{Title}"
    textTransformPolicy=ToUpper
    font {
     bForceMonospaced=false
     bMaterialIsStencil=false
     fontMaterial=null
     fontObject=&BurbankSmall
     letterSpacing=0
     monospacedWidth=1
     size=24
     skewAmount=0
     typefaceFontName=Black
     outlineSettings {
      bApplyOutlineToDropShadows=false
      bMiteredCorners=false
      bSeparateFillAlpha=false
      outlineMaterial=null
      outlineSize=0
      outlineColor=<rgba 0 0 0 1>
     }
    }
    shadowOffset=<0 0>
   }
   slot {class=VerticalBoxSlot, properties {padding=<ltrb 0 0 0 8>}}
  }
  CommonText_Description {
   widgetClassPath=&CommonTextBlock
   bInherited=true
   parent=VerticalBox_0
   widget=CommonText_Description
   properties {
    text="{Description}"
    wrapTextAt=500
    colorAndOpacity {colorUseRule=UseColor_Specified, specifiedColor=<rgba 0.2159 0.6584 1 1>}
    font {
     bForceMonospaced=false
     bMaterialIsStencil=false
     fontMaterial=null
     fontObject=&BurbankSmall
     letterSpacing=0
     monospacedWidth=1
     size=24
     skewAmount=0
     typefaceFontName=Black
     outlineSettings {
      bApplyOutlineToDropShadows=false
      bMiteredCorners=false
      bSeparateFillAlpha=false
      outlineMaterial=null
      outlineSize=0
      outlineColor=<rgba 0 0 0 1>
     }
    }
    shadowOffset=<0 0>
   }
   slot {class=VerticalBoxSlot}
  }
  Overlay_1 {
   widgetClassPath=&Overlay
   parent=HorizontalBox_0
   widget=Overlay_1
   slot {class=HorizontalBoxSlot, properties {horizontalAlignment=HAlign_Right, verticalAlignment=VAlign_Center, padding=<ltrb 30 20 20 20>}}
  }
  SizeBox_81 {
   widgetClassPath=&SizeBox
   parent=Overlay_1
   widget=SizeBox_81
   properties {bOverride_HeightOverride=true, bOverride_WidthOverride=true, heightOverride=75, widthOverride=75}
   slot {class=OverlaySlot, properties {horizontalAlignment=HAlign_Right, verticalAlignment=VAlign_Center}}
  }
  Image_Warning {
   widgetClassPath=&Image
   bIsVariable=true
   parent=SizeBox_81
   widget=Image_Warning
   properties {
    brush {
     bIsDynamicallyLoaded=false
     drawAs=Image
     imageType=NoImage
     mirroring=NoMirror
     resourceName=None
     resourceObject=/Game/Athena/HUD/Bacchus/HUDLayoutTool/HLT_Icons/T-HLT-Warning-64_Sprite
     tiling=NoTile
     imageSize=<64 64>
     margin=<ltrb 0 0 0 0>
     outlineSettings {
      bUseBrushTransparency=false
      roundingType=HalfHeightRadius
      width=0
      color {colorUseRule=UseColor_Specified, specifiedColor=<rgba 0 0 0 0>}
      cornerRadii=<0 0 0 0>
     }
     tintColor {colorUseRule=UseColor_Specified, specifiedColor=<rgba 1 1 1 1>}
     uVRegion {bIsValid=false, max=<0 0>, min=<0 0>}
    }
    colorAndOpacity=<rgba 0.9301 1 0.0097 1>
   }
   slot {class=SizeBoxSlot}
  }
  SizeBox_18 {
   widgetClassPath=&SizeBox
   parent=Overlay_1
   widget=SizeBox_18
   properties {bOverride_HeightOverride=true, bOverride_WidthOverride=true, heightOverride=75, widthOverride=72}
   slot {class=OverlaySlot, properties {horizontalAlignment=HAlign_Right, verticalAlignment=VAlign_Center}}
  }
  Image_Checkmark {
   widgetClassPath=&Image
   parent=SizeBox_18
   widget=Image_Checkmark
   properties {
    brush {
     bIsDynamicallyLoaded=false
     drawAs=Image
     imageType=NoImage
     mirroring=NoMirror
     resourceName=None
     resourceObject=/Game/Athena/HUD/Bacchus/HUDLayoutTool/HLT_Icons/T-HLT-Checkmark-64_Sprite
     tiling=NoTile
     imageSize=<72 75>
     margin=<ltrb 0 0 0 0>
     outlineSettings {
      bUseBrushTransparency=false
      roundingType=HalfHeightRadius
      width=0
      color {colorUseRule=UseColor_Specified, specifiedColor=<rgba 0 0 0 0>}
      cornerRadii=<0 0 0 0>
     }
     tintColor {colorUseRule=UseColor_Specified, specifiedColor=<rgba 1 1 1 1>}
     uVRegion {bIsValid=false, max=<0 0>, min=<0 0>}
    }
    colorAndOpacity=<rgba 0.0343 1 0.227 1>
   }
   slot {class=SizeBoxSlot, properties {horizontalAlignment=HAlign_Center, verticalAlignment=VAlign_Center}}
  }
 ]
}
