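# HUDLayoutToolV2_TooltipLayer is the Widget Blueprint for the tooltip layer of the mobile (Bacchus) HUD Layout Tool.
# A root canvas panel holds one auto-sized, top-left anchored HUDLayoutToolV2_Tooltip, bound to the native parent class as Tooltip_Message.
# The event graph has only unconnected stock events, so any tooltip behavior is left to the C++ parent and the tooltip widget.
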WidgetBlueprint /FortniteCoreUI/Athena/HUD/Bacchus/HUDLayoutTool/HUDLayoutToolV2_TooltipLayer
graphs [EventGraph {nodes [
   Event PreConstruct()
   Event Construct()
   Event Tick()
  ]}]
variables []
widgets {
 info {inheritedWidgetCount=1, parentClass=/Script/FortniteUI.HUDLayoutToolV2_TooltipLayer, rootWidgetClass=&CanvasPanel, widgetCount=2}
 widgets [
  CanvasPanel_110 {widgetClassPath=&CanvasPanel, widget=CanvasPanel_110, properties {visibility=Visible}}
  Tooltip_Message {
   widgetClassPath=&/HUDLayoutToolV2_Tooltip
   bInherited=true
   bIsVariable=true
   parent=CanvasPanel_110
   widget=Tooltip_Message
   properties {visibility=Visible}
   slot {
    class=CanvasPanelSlot
    properties {bAutoSize=true, layoutData {alignment=<0 0>, anchors {maximum=<0 0>, minimum=<0 0>}, offsets=<ltrb 0 0 200 200>}}
   }
  }
 ]
}
