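# Widget catalog panel for the mobile HUD layout tool (V2), listing button and HUD element previews.
# Two toggle tabs, Buttons and HUD, sit above a layout icon, context name and horizontally scrolling ThresholdScrollBox.
# Inside it, an activatable switcher holds two dynamic entry boxes of WidgetPreview entries.
# The Blueprint supplies only layout and styling; its graph is empty, with key widgets inherited from the native FortMobileHUDLayoutToolV2_WidgetCatalog parent.
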WidgetBlueprint /FortniteCoreUI/Athena/HUD/Bacchus/HUDLayoutTool/HUDLayoutToolV2_WidgetCatalog
graphs [EventGraph {links [], nodes []}]
variables []
widgets {
 info {
  inheritedWidgetCount=7
  parentClass=/Script/FortniteUI.FortMobileHUDLayoutToolV2_WidgetCatalog
  rootWidgetClass=&Overlay
  widgetCount=25
 }
 widgets [
  Overlay_4 {widgetClassPath=&Overlay, widget=Overlay_4}
  Overlay_0 {
   widgetClassPath=&Overlay
   parent=Overlay_4
   widget=Overlay_0
   slot {class=OverlaySlot, properties {horizontalAlignment=HAlign_Fill, verticalAlignment=VAlign_Fill, padding=<ltrb 0 100 0 0>}}
  }
  Background {
   widgetClassPath=&Image
   parent=Overlay_0
   widget=Background
   properties {colorAndOpacity=<rgba 0.0232 0.0232 0.0232 1>}
   slot {class=OverlaySlot, properties {horizontalAlignment=HAlign_Fill, verticalAlignment=VAlign_Fill}}
  }
  VerticalBox_0 {
   widgetClassPath=&VerticalBox
   parent=Overlay_4
   widget=VerticalBox_0
   slot {class=OverlaySlot, properties {horizontalAlignment=HAlign_Fill, verticalAlignment=VAlign_Fill}}
  }
  SafeZone_6 {
   widgetClassPath=&SafeZone
   parent=VerticalBox_0
   widget=SafeZone_6
   properties {padBottom=false, padRight=false, padTop=false}
   slot {class=VerticalBoxSlot, properties {horizontalAlignment=HAlign_Left}}
  }
  SizeBox_1 {
   widgetClassPath=&SizeBox
   parent=SafeZone_6
   widget=SizeBox_1
   properties {bOverride_HeightOverride=true, heightOverride=100}
   slot {class=SafeZoneSlot}
  }
  Overlay_1 {widgetClassPath=&Overlay, parent=SizeBox_1, widget=Overlay_1, slot {class=SizeBoxSlot}}
  TabsBackground {
   widgetClassPath=&Image
   parent=Overlay_1
   widget=TabsBackground
   properties {colorAndOpacity=<rgba 0.0232 0.0232 0.0232 1>}
   slot {class=OverlaySlot, properties {horizontalAlignment=HAlign_Fill, verticalAlignment=VAlign_Fill}}
  }
  HorizontalBox_111 {
   widgetClassPath=&HorizontalBox
   parent=Overlay_1
   widget=HorizontalBox_111
   slot {class=OverlaySlot, properties {horizontalAlignment=HAlign_Fill, verticalAlignment=VAlign_Center, padding=<ltrb 20 10 10 10>}}
  }
  Button_DisplayButtonPreviews {
   widgetClassPath=/Game/Athena/HUD/Bacchus/HUDLayoutTool/HUDLayoutToolV2_Button
   bInherited=true
   parent=HorizontalBox_111
   widget=Button_DisplayButtonPreviews
   properties {
    bIsToggleable=true
    bOverride_HeightOverride=true
    bOverride_WidthOverride=true
    buttonHeightOverride=76
    buttonWidthOverride=212
    text=Buttons
    iconBrush {
     bIsDynamicallyLoaded=false
     drawAs=Image
     imageType=NoImage
     mirroring=NoMirror
     resourceName=None
     resourceObject=/Game/Athena/HUD/Bacchus/HUDLayoutTool/HLT_Icons/T-HLT-CatalogButton_Sprite
     tiling=NoTile
     imageSize=<50 50>
     margin=<ltrb 0 0 0 0>
     outlineSettings {
      bUseBrushTransparency=false
      roundingType=HalfHeightRadius
      width=0
      color {colorUseRule=UseColor_Specified, specifiedColor=<rgba 0 0 0 0>}
      cornerRadii=<0 0 0 0>
     }
     tintColor {colorUseRule=UseColor_Specified, specifiedColor=<rgba 1 1 1 1>}
     uVRegion {bIsValid=false, max=<0 0>, min=<0 0>}
    }
    internalPadding=<ltrb 0 0 0 0>
    normalButtonStyle {
     iconAngle=0
     backgroundColor=<rgba 0 0.117 1 0.5>
     iconColor=<rgba 1 1 1 0.5>
     textColor {colorUseRule=UseColor_Specified, specifiedColor=<rgba 1 1 1 0.5>}
    }
    selectedButtonStyle {
     iconAngle=0
     backgroundColor=<rgba 0 0.1137 1 1>
     iconColor=<rgba 1 1 1 1>
     textColor {colorUseRule=UseColor_Specified, specifiedColor=<rgba 1 1 1 1>}
    }
    textPadding=<ltrb 0 0 10 0>
   }
   slot {class=HorizontalBoxSlot, properties {padding=<ltrb 0 0 20 0>}}
  }
  Button_DisplayWidgetPreviews {
   widgetClassPath=&HUDLayoutToolV2_Button
   bInherited=true
   parent=HorizontalBox_111
   widget=Button_DisplayWidgetPreviews
   properties {
    bIsToggleable=true
    bOverride_HeightOverride=true
    bOverride_WidthOverride=true
    buttonHeightOverride=76
    buttonWidthOverride=212
    text=HUD
    iconBrush {
     bIsDynamicallyLoaded=false
     drawAs=Image
     imageType=NoImage
     mirroring=NoMirror
     resourceName=None
     resourceObject=/Game/Athena/HUD/Bacchus/HUDLayoutTool/HLT_Icons/T-HLT-CatalogWidget_Sprite
     tiling=NoTile
     imageSize=<102 44>
     margin=<ltrb 0 0 0 0>
     outlineSettings {
      bUseBrushTransparency=false
      roundingType=HalfHeightRadius
      width=0
      color {colorUseRule=UseColor_Specified, specifiedColor=<rgba 0 0 0 0>}
      cornerRadii=<0 0 0 0>
     }
     tintColor {colorUseRule=UseColor_Specified, specifiedColor=<rgba 1 1 1 1>}
     uVRegion {bIsValid=false, max=<0 0>, min=<0 0>}
    }
    internalPadding=<ltrb 0 0 0 0>
    normalButtonStyle {
     iconAngle=0
     backgroundColor=<rgba 0 0.117 1 0.498>
     iconColor=<rgba 1 1 1 0.5>
     textColor {colorUseRule=UseColor_Specified, specifiedColor=<rgba 1 1 1 0.5>}
    }
    selectedButtonStyle {
     iconAngle=0
     backgroundColor=<rgba 0 0.1137 1 1>
     iconColor=<rgba 1 1 1 1>
     textColor {colorUseRule=UseColor_Specified, specifiedColor=<rgba 1 1 1 1>}
    }
    textPadding=<ltrb 0 0 10 0>
   }
   slot {class=HorizontalBoxSlot}
  }
  SafeZone_5 {
   widgetClassPath=&SafeZone
   parent=VerticalBox_0
   widget=SafeZone_5
   properties {padTop=false}
   slot {class=VerticalBoxSlot, properties {verticalAlignment=VAlign_Bottom, padding=<ltrb 20 6 20 20>}}
  }
  SizeBox_0 {
   widgetClassPath=&SizeBox
   parent=SafeZone_5
   widget=SizeBox_0
   properties {bOverride_MinDesiredHeight=true, heightOverride=180, maxDesiredHeight=160, minDesiredHeight=180}
   slot {class=SafeZoneSlot}
  }
  Overlay_2 {widgetClassPath=&Overlay, parent=SizeBox_0, widget=Overlay_2, slot {class=SizeBoxSlot}}
  VerticalBox_1 {
   widgetClassPath=&VerticalBox
   parent=Overlay_2
   widget=VerticalBox_1
   slot {class=OverlaySlot, properties {horizontalAlignment=HAlign_Fill, verticalAlignment=VAlign_Fill}}
  }
  HorizontalBox_0 {
   widgetClassPath=&HorizontalBox
   parent=VerticalBox_1
   widget=HorizontalBox_0
   slot {class=VerticalBoxSlot, properties {horizontalAlignment=HAlign_Left, padding=<ltrb 0 0 0 10>}}
  }
  SizeBox_2 {
   widgetClassPath=&SizeBox
   parent=HorizontalBox_0
   widget=SizeBox_2
   properties {bOverride_HeightOverride=true, heightOverride=42, minDesiredWidth=42}
   slot {class=HorizontalBoxSlot, properties {padding=<ltrb 0 0 10 0>, size {sizeRule=Fill, value=1}}}
  }
  ScaleBox_0 {widgetClassPath=&ScaleBox, parent=SizeBox_2, widget=ScaleBox_0, slot {class=SizeBoxSlot}}
  Image_LayoutIcon {
   widgetClassPath=&Image
   bInherited=true
   bIsVariable=true
   parent=ScaleBox_0
   widget=Image_LayoutIcon
   properties {brush {
     bIsDynamicallyLoaded=false
     drawAs=Image
     imageType=NoImage
     mirroring=NoMirror
     resourceName=None
     resourceObject=null
     tiling=NoTile
     imageSize=<57 57>
     margin=<ltrb 0 0 0 0>
     outlineSettings {
      bUseBrushTransparency=false
      roundingType=HalfHeightRadius
      width=0
      color {colorUseRule=UseColor_Specified, specifiedColor=<rgba 0 0 0 0>}
      cornerRadii=<0 0 0 0>
     }
     tintColor {colorUseRule=UseColor_Specified, specifiedColor=<rgba 1 1 1 1>}
     uVRegion {bIsValid=false, max=<0 0>, min=<0 0>}
    }}
   slot {class=ScaleBoxSlot}
  }
  Text_ContextName {
   widgetClassPath=&CommonTextBlock
   bInherited=true
   parent=HorizontalBox_0
   widget=Text_ContextName
   properties {
    scrollStyle=&TextScrollStyle-Base_C
    text="{Current Layout Name}"
    colorAndOpacity {colorUseRule=UseColor_Specified, specifiedColor=<rgba 0.4125 0.4125 0.4125 1>}
    font {
     bForceMonospaced=false
     bMaterialIsStencil=false
     fontMaterial=null
     fontObject=&BurbankSmall
     letterSpacing=0
     monospacedWidth=1
     size=22
     skewAmount=0
     typefaceFontName=Bold
     outlineSettings {
      bApplyOutlineToDropShadows=false
      bMiteredCorners=false
      bSeparateFillAlpha=false
      outlineMaterial=null
      outlineSize=0
      outlineColor=<rgba 0 0 0 1>
     }
    }
    shadowOffset=<0 0>
   }
   slot {class=HorizontalBoxSlot}
  }
  ThresholdScrollBox_0 {
   widgetClassPath=/Script/FortniteUI.ThresholdScrollBox
   parent=VerticalBox_1
   widget=ThresholdScrollBox_0
   properties {alwaysShowScrollbar=true, orientation=Orient_Horizontal, scrollThreshold=50, scrollbarPadding=<ltrb 0 20 0 0>}
   slot {class=VerticalBoxSlot, properties {verticalAlignment=VAlign_Top}}
  }
  SizeBox_7 {
   widgetClassPath=&SizeBox
   parent=ThresholdScrollBox_0
   widget=SizeBox_7
   properties {bOverride_HeightOverride=true, heightOverride=128}
   slot {class=ScrollBoxSlot}
  }
  Switcher_Previews {
   widgetClassPath=&CommonActivatableWidgetSwitcher
   bInherited=true
   parent=SizeBox_7
   widget=Switcher_Previews
   slot {class=SizeBoxSlot}
  }
  EntryBox_ButtonPreviews {
   widgetClassPath=/Script/FortniteUI.FortDynamicEntryBox
   bInherited=true
   parent=Switcher_Previews
   widget=EntryBox_ButtonPreviews
   properties {
    bOverrideAccessibleDefaults=true
    entryHorizontalAlignment=HAlign_Left
    entryVerticalAlignment=VAlign_Center
    entryWidgetClass=&/HUDLayoutToolV2_WidgetPreview.HUDLayoutToolV2_WidgetPreview_C
    numDesignerPreviewEntries=5
    entrySpacing=<40 0>
   }
   slot {class=WidgetSwitcherSlot}
  }
  EntryBox_WidgetPreviews {
   widgetClassPath=&FortDynamicEntryBox
   bInherited=true
   parent=Switcher_Previews
   widget=EntryBox_WidgetPreviews
   properties {
    bOverrideAccessibleDefaults=true
    entryHorizontalAlignment=HAlign_Left
    entryVerticalAlignment=VAlign_Center
    entryWidgetClass=&HUDLayoutToolV2_WidgetPreview_C
    numDesignerPreviewEntries=4
    entrySpacing=<40 0>
   }
   slot {class=WidgetSwitcherSlot}
  }
 ]
}
