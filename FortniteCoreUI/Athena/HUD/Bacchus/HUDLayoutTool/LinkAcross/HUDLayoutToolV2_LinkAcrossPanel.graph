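# Modal Link Across panel for the HUD Layout Tool, a centered window over a dimmed blocker for confirming or cancelling changes to links between HUD element copies across modes.
# A header holds the title, close button and a collapsed max-copies warning.
# A visibility switcher holds the link action panel and a simple text prompt, with an "Edit links" toggle below.
# Logic lives in the C++ parent; the event graph has only unconnected stub events.
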WidgetBlueprint /FortniteCoreUI/Athena/HUD/Bacchus/HUDLayoutTool/LinkAcross/HUDLayoutToolV2_LinkAcrossPanel
graphs [EventGraph {nodes [
   Event PreConstruct()
   Event Construct()
   Event Tick()
  ]}]
variables []
widgets {
 info {inheritedWidgetCount=12, parentClass=/Script/FortniteUI.HUDLayoutToolV2_LinkAcrossPanel, rootWidgetClass=&Overlay, widgetCount=24}
 widgets [
  Overlay_1 {widgetClassPath=&Overlay, widget=Overlay_1}
  Image_TransparentBlocker {
   widgetClassPath=&Image
   parent=Overlay_1
   widget=Image_TransparentBlocker
   properties {
    brush {
     bIsDynamicallyLoaded=false
     drawAs=Image
     imageType=NoImage
     mirroring=NoMirror
     resourceName=None
     resourceObject=null
     tiling=NoTile
     imageSize=<64 64>
     margin=<ltrb 0 0 0 0>
     outlineSettings {
      bUseBrushTransparency=false
      roundingType=HalfHeightRadius
      width=0
      color {colorUseRule=UseColor_Specified, specifiedColor=<rgba 0 0 0 0>}
      cornerRadii=<0 0 0 0>
     }
     tintColor {colorUseRule=UseColor_Specified, specifiedColor=<rgba 1 1 1 1>}
     uVRegion {bIsValid=false, max=<0 0>, min=<0 0>}
    }
    colorAndOpacity=<rgba 0 0 0 0.3>
   }
   slot {class=OverlaySlot, properties {horizontalAlignment=HAlign_Fill, verticalAlignment=VAlign_Fill}}
  }
  SizeBox_0 {
   widgetClassPath=&SizeBox
   parent=Overlay_1
   widget=SizeBox_0
   properties {bOverride_WidthOverride=true, heightOverride=800, maxDesiredWidth=800, minDesiredHeight=470, widthOverride=1000}
   slot {class=OverlaySlot, properties {horizontalAlignment=HAlign_Center, verticalAlignment=VAlign_Center}}
  }
  Overlay_17 {widgetClassPath=&Overlay, parent=SizeBox_0, widget=Overlay_17, slot {class=SizeBoxSlot}}
  WindowBackground {
   widgetClassPath=&Image
   parent=Overlay_17
   widget=WindowBackground
   properties {
    brush {
     bIsDynamicallyLoaded=false
     drawAs=Image
     imageType=NoImage
     mirroring=NoMirror
     resourceName=None
     resourceObject=null
     tiling=NoTile
     imageSize=<64 64>
     margin=<ltrb 0 0 0 0>
     outlineSettings {
      bUseBrushTransparency=false
      roundingType=HalfHeightRadius
      width=0
      color {colorUseRule=UseColor_Specified, specifiedColor=<rgba 0 0 0 0>}
      cornerRadii=<0 0 0 0>
     }
     tintColor {colorUseRule=UseColor_Specified, specifiedColor=<rgba 1 1 1 1>}
     uVRegion {bIsValid=false, max=<0 0>, min=<0 0>}
    }
    colorAndOpacity=<rgba 0.0561 0.0561 0.0561 1>
   }
   slot {class=OverlaySlot, properties {horizontalAlignment=HAlign_Fill, verticalAlignment=VAlign_Fill}}
  }
  VerticalBox_0 {
   widgetClassPath=&VerticalBox
   parent=Overlay_17
   widget=VerticalBox_0
   slot {class=OverlaySlot, properties {horizontalAlignment=HAlign_Fill, verticalAlignment=VAlign_Fill, padding=<ltrb 20 20 20 20>}}
  }
  Overlay_0 {
   widgetClassPath=&Overlay
   parent=VerticalBox_0
   widget=Overlay_0
   slot {class=VerticalBoxSlot, properties {verticalAlignment=VAlign_Top, size {sizeRule=Automatic, value=0.2}}}
  }
  HorizontalBox_0 {
   widgetClassPath=&HorizontalBox
   parent=Overlay_0
   widget=HorizontalBox_0
   slot {class=OverlaySlot, properties {horizontalAlignment=HAlign_Fill, verticalAlignment=VAlign_Center, padding=<ltrb 0 0 135 10>}}
  }
  LinkAcrossPanelTitle_Header {
   widgetClassPath=&/HUDLayoutToolV2_LinkAcrossPanelTitle
   bInherited=true
   bIsVariable=true
   parent=HorizontalBox_0
   widget=LinkAcrossPanelTitle_Header
   slot {class=HorizontalBoxSlot, properties {size {sizeRule=Fill, value=1}}}
  }
  SizeBox_1 {
   widgetClassPath=&SizeBox
   parent=HorizontalBox_0
   widget=SizeBox_1
   properties {bOverride_MinDesiredHeight=true, bOverride_MinDesiredWidth=true, minDesiredHeight=80, minDesiredWidth=100}
   slot {class=HorizontalBoxSlot, properties {horizontalAlignment=HAlign_Right, verticalAlignment=VAlign_Center}}
  }
  WarningIndicator_MainWarning {
   widgetClassPath=/Game/Athena/HUD/Bacchus/HUDLayoutTool/LinkAcross/HUDLayoutToolV2_LinkAcrossWarningIndicator
   bInherited=true
   bIsVariable=true
   parent=SizeBox_1
   widget=WarningIndicator_MainWarning
   properties {
    visibility=Collapsed
    warningMessage=This HUD element has reached its maximum number of copies in a mode. You must remove one to add a new copy.
    colorAndOpacity=<rgba 0.9301 1 0.0097 1>
   }
   slot {class=SizeBoxSlot, properties {padding=<ltrb 20 0 0 0>}}
  }
  Button_Close {
   widgetClassPath=/FortUILibrary/Foundation/Buttons/CloseButton
   bInherited=true
   parent=Overlay_0
   widget=Button_Close
   properties {
    onButtonBaseClicked=(&self:WidgetTree.Button_Close.HandleOnButtonClicked)
    onButtonBaseDoubleClicked=(&self:WidgetTree.Button_Close.HandleOnButtonDoubleClicked)
    onButtonBaseHovered=(&self:WidgetTree.Button_Close.HandleOnButtonHovered)
    onButtonBaseUnhovered=(&self:WidgetTree.Button_Close.HandleOnButtonUnhovered)
    onSelectedChangedBase=(&self:WidgetTree.Button_Close.HandleOnSelectedChanged)
   }
   slot {class=OverlaySlot, properties {horizontalAlignment=HAlign_Right, padding=<ltrb 0 0 0 20>}}
  }
  Separator {
   widgetClassPath=&Border
   parent=Overlay_0
   widget=Separator
   properties {
    background {
     bIsDynamicallyLoaded=false
     drawAs=Image
     imageType=NoImage
     mirroring=NoMirror
     resourceName=None
     resourceObject=null
     tiling=NoTile
     imageSize=<32 32>
     margin=<ltrb 0 0 0 0>
     outlineSettings {
      bUseBrushTransparency=false
      roundingType=HalfHeightRadius
      width=0
      color {colorUseRule=UseColor_Specified, specifiedColor=<rgba 0 0 0 0>}
      cornerRadii=<0 0 0 0>
     }
     tintColor {colorUseRule=UseColor_Specified, specifiedColor=<rgba 1 1 1 0.3>}
     uVRegion {bIsValid=false, max=<0 0>, min=<0 0>}
    }
    padding=<ltrb 0 2 0 2>
   }
   slot {class=OverlaySlot, properties {horizontalAlignment=HAlign_Fill, verticalAlignment=VAlign_Bottom}}
  }
  Switcher_Actions {
   widgetClassPath=&CommonVisibilitySwitcher
   bInherited=true
   parent=VerticalBox_0
   widget=Switcher_Actions
   slot {class=VerticalBoxSlot, properties {padding=<ltrb 0 20 0 20>, size {sizeRule=Fill, value=1}}}
  }
  SizeBox_ActionPanelContainer {
   widgetClassPath=&SizeBox
   bInherited=true
   parent=Switcher_Actions
   widget=SizeBox_ActionPanelContainer
   properties {bOverride_HeightOverride=true, heightOverride=520, minDesiredHeight=520}
   slot {class=CommonVisibilitySwitcherSlot}
  }
  LinkAcrossActionPanel_Actions {
   widgetClassPath=&/HUDLayoutToolV2_LinkAcrossActionPanel
   bInherited=true
   bIsVariable=true
   parent=SizeBox_ActionPanelContainer
   widget=LinkAcrossActionPanel_Actions
   slot {class=SizeBoxSlot}
  }
  VerticalBox_SimpleAction {
   widgetClassPath=&VerticalBox
   bInherited=true
   parent=Switcher_Actions
   widget=VerticalBox_SimpleAction
   slot {class=CommonVisibilitySwitcherSlot, properties {verticalAlignment=VAlign_Center, padding=<ltrb 0 40 0 44>}}
  }
  Text_SimpleAction {
   widgetClassPath=&CommonTextBlock
   bInherited=true
   parent=VerticalBox_SimpleAction
   widget=Text_SimpleAction
   properties {
    autoWrapText=true
    justification=Center
    text=Simplified Action Really Long Text that Is Also A question? To know what it looks like if we ask a lot of questions in a link across window.
    font {
     bForceMonospaced=false
     bMaterialIsStencil=false
     fontMaterial=null
     fontObject=&BurbankSmall
     letterSpacing=0
     monospacedWidth=1
     size=30
     skewAmount=0
     typefaceFontName=Bold
     outlineSettings {
      bApplyOutlineToDropShadows=false
      bMiteredCorners=false
      bSeparateFillAlpha=false
      outlineMaterial=null
      outlineSize=0
      outlineColor=<rgba 0 0 0 1>
     }
    }
    shadowOffset=<0 0>
   }
   slot {class=VerticalBoxSlot}
  }
  Text_SimpleActionWarning {
   widgetClassPath=&CommonTextBlock
   bInherited=true
   parent=VerticalBox_SimpleAction
   widget=Text_SimpleActionWarning
   properties {
    autoWrapText=true
    justification=Center
    text=(Changes in advanced section won't be applied.)
    visibility=Collapsed
    colorAndOpacity {colorUseRule=UseColor_Specified, specifiedColor=<rgba 0.9301 1 0.0097 1>}
    font {
     bForceMonospaced=false
     bMaterialIsStencil=false
     fontMaterial=null
     fontObject=&BurbankSmall
     letterSpacing=0
     monospacedWidth=1
     size=22
     skewAmount=0
     typefaceFontName=Black
     outlineSettings {
      bApplyOutlineToDropShadows=false
      bMiteredCorners=false
      bSeparateFillAlpha=false
      outlineMaterial=null
      outlineSize=0
      outlineColor=<rgba 0 0 0 1>
     }
    }
    shadowOffset=<0 0>
   }
   slot {class=VerticalBoxSlot, properties {padding=<ltrb 0 10 0 0>}}
  }
  Overlay_2 {
   widgetClassPath=&Overlay
   parent=VerticalBox_0
   widget=Overlay_2
   slot {class=VerticalBoxSlot, properties {verticalAlignment=VAlign_Bottom}}
  }
  Button_ToggleAdvanced {
   widgetClassPath=/Game/Athena/HUD/Bacchus/HUDLayoutTool/HUDLayoutToolV2_Button
   bInherited=true
   parent=Overlay_2
   widget=Button_ToggleAdvanced
   properties {
    bIsToggleable=true
    bOverride_HeightOverride=true
    bOverride_WidthOverride=true
    buttonHeightOverride=80
    buttonWidthOverride=300
    text=Edit links
    textFontSize=24
    backgroundBrush {
     bIsDynamicallyLoaded=false
     drawAs=NoDrawType
     imageType=NoImage
     mirroring=NoMirror
     resourceName=None
     resourceObject=null
     tiling=NoTile
     imageSize=<32 32>
     margin=<ltrb 0 0 0 0>
     outlineSettings {
      bUseBrushTransparency=false
      roundingType=HalfHeightRadius
      width=0
      color {colorUseRule=UseColor_Specified, specifiedColor=<rgba 0 0 0 0>}
      cornerRadii=<0 0 0 0>
     }
     tintColor {colorUseRule=UseColor_Specified, specifiedColor=<rgba 1 1 1 1>}
     uVRegion {bIsValid=false, max=<0 0>, min=<0 0>}
    }
    iconBrush {
     bIsDynamicallyLoaded=false
     drawAs=Image
     imageType=NoImage
     mirroring=NoMirror
     resourceName=None
     resourceObject=/Game/Athena/HUD/Bacchus/HUDLayoutTool/HLT_Icons/T-HLT-Triangle_Sprite
     tiling=NoTile
     imageSize=<30 26>
     margin=<ltrb 0 0 0 0>
     outlineSettings {
      bUseBrushTransparency=false
      roundingType=HalfHeightRadius
      width=0
      color {colorUseRule=UseColor_Specified, specifiedColor=<rgba 0 0 0 0>}
      cornerRadii=<0 0 0 0>
     }
     tintColor {colorUseRule=UseColor_Specified, specifiedColor=<rgba 1 1 1 1>}
     uVRegion {bIsValid=false, max=<0 0>, min=<0 0>}
    }
    iconPadding=<ltrb 20 0 0 0>
    internalPadding=<ltrb 0 0 0 0>
    normalButtonStyle {
     iconAngle=180
     backgroundColor=<rgba 0 0.117 1 1>
     iconColor=<rgba 1 1 1 1>
     textColor {colorUseRule=UseColor_Specified, specifiedColor=<rgba 1 1 1 1>}
    }
    selectedButtonStyle {
     iconAngle=0
     backgroundColor=<rgba 0 0.117 1 1>
     iconColor=<rgba 1 1 1 1>
     textColor {colorUseRule=UseColor_Specified, specifiedColor=<rgba 1 1 1 1>}
    }
   }
   slot {class=OverlaySlot, properties {verticalAlignment=VAlign_Center}}
  }
  HorizontalBox_248 {
   widgetClassPath=&HorizontalBox
   parent=Overlay_2
   widget=HorizontalBox_248
   slot {class=OverlaySlot, properties {horizontalAlignment=HAlign_Right, verticalAlignment=VAlign_Center}}
  }
  Button_Confirm {
   widgetClassPath=&HUDLayoutToolV2_Button
   bInherited=true
   parent=HorizontalBox_248
   widget=Button_Confirm
   properties {
    bOverride_HeightOverride=true
    bOverride_WidthOverride=true
    buttonHeightOverride=80
    buttonWidthOverride=246
    text=Confirm
    textFontSize=26
    internalPadding=<ltrb 24 12 24 12>
   }
   slot {class=HorizontalBoxSlot, properties {horizontalAlignment=HAlign_Right, verticalAlignment=VAlign_Center, padding=<ltrb 20 0 0 0>}}
  }
  Button_Cancel {
   widgetClassPath=&HUDLayoutToolV2_Button
   bInherited=true
   parent=HorizontalBox_248
   widget=Button_Cancel
   properties {
    bOverride_HeightOverride=true
    bOverride_WidthOverride=true
    buttonHeightOverride=80
    buttonWidthOverride=246
    text=Cancel
    textFontSize=26
    internalPadding=<ltrb 24 12 24 12>
    normalButtonStyle {
     iconAngle=0
     backgroundColor=<rgba 0.9301 1 0.0097 1>
     iconColor=<rgba 0 0 0 1>
     textColor {colorUseRule=UseColor_Specified, specifiedColor=<rgba 0 0 0 1>}
    }
    pressedButtonStyle {
     iconAngle=0
     backgroundColor=<rgba 1 1 1 1>
     iconColor=<rgba 0 0 0 1>
     textColor {colorUseRule=UseColor_Specified, specifiedColor=<rgba 0 0 0 1>}
    }
   }
   slot {class=HorizontalBoxSlot, properties {horizontalAlignment=HAlign_Right, verticalAlignment=VAlign_Center, padding=<ltrb 20 0 0 0>}}
  }
 ]
}
