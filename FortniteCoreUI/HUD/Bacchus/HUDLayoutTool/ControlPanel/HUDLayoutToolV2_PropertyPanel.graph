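# Widget Blueprint for the property panel of the HUD Layout Tool V2, laying out the widgets its native parent class declares, with no graph logic.
# An activatable switcher holds a "Select a widget to edit" prompt and a properties container.
# The container's header shows the widget name beside Edit Links, revert and red delete buttons.
# Below it, a settings list view takes its visual data from a visual settings registry.
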WidgetBlueprint /FortniteCoreUI/HUD/Bacchus/HUDLayoutTool/ControlPanel/HUDLayoutToolV2_PropertyPanel
graphs [EventGraph {links [], nodes []}]
variables []
widgets {
 info {inheritedWidgetCount=8, parentClass=/Script/FortniteUI.HUDLayoutToolV2_PropertyPanel, rootWidgetClass=&SizeBox, widgetCount=12}
 widgets [
  SizeBox_0 {
   widgetClassPath=&SizeBox
   widget=SizeBox_0
   properties {bOverride_HeightOverride=true, bOverride_WidthOverride=true, heightOverride=500, widthOverride=934}
  }
  Switcher_PropertiesAndMessage {
   widgetClassPath=&CommonActivatableWidgetSwitcher
   bInherited=true
   parent=SizeBox_0
   widget=Switcher_PropertiesAndMessage
   slot {class=SizeBoxSlot, properties {padding=<ltrb 20 10 20 20>}}
  }
  Text_SelectWidgetMessage {
   widgetClassPath=&CommonTextBlock
   bInherited=true
   parent=Switcher_PropertiesAndMessage
   widget=Text_SelectWidgetMessage
   properties {
    justification=Center
    text=Select a widget to edit
    textTransformPolicy=ToUpper
    colorAndOpacity {colorUseRule=UseColor_Specified, specifiedColor=<rgba 1 1 1 0.5>}
    font {
     bForceMonospaced=false
     bMaterialIsStencil=false
     fontMaterial=null
     fontObject=&BurbankSmall
     letterSpacing=0
     monospacedWidth=1
     size=24
     skewAmount=0
     typefaceFontName=Black
     outlineSettings {
      bApplyOutlineToDropShadows=false
      bMiteredCorners=false
      bSeparateFillAlpha=false
      outlineMaterial=null
      outlineSize=0
      outlineColor=<rgba 0 0 0 1>
     }
    }
    shadowOffset=<0 0>
   }
   slot {class=WidgetSwitcherSlot, properties {horizontalAlignment=HAlign_Center, verticalAlignment=VAlign_Center}}
  }
  VerticalBox_PropertiesContainer {
   widgetClassPath=&VerticalBox
   bInherited=true
   parent=Switcher_PropertiesAndMessage
   widget=VerticalBox_PropertiesContainer
   slot {class=WidgetSwitcherSlot}
  }
  Overlay_0 {
   widgetClassPath=&Overlay
   parent=VerticalBox_PropertiesContainer
   widget=Overlay_0
   slot {class=VerticalBoxSlot, properties {padding=<ltrb 0 0 0 10>}}
  }
  SizeBox_3 {
   widgetClassPath=&SizeBox
   parent=Overlay_0
   widget=SizeBox_3
   properties {bOverride_WidthOverride=true, widthOverride=475}
   slot {class=OverlaySlot, properties {verticalAlignment=VAlign_Center, padding=<ltrb 15 0 0 0>}}
  }
  Text_WidgetName {
   widgetClassPath=&CommonTextBlock
   bInherited=true
   parent=SizeBox_3
   widget=Text_WidgetName
   properties {
    scrollStyle=&TextScrollStyle-Base_C
    text=Widget Name
    textTransformPolicy=ToUpper
    font {
     bForceMonospaced=false
     bMaterialIsStencil=false
     fontMaterial=null
     fontObject=&BurbankSmall
     letterSpacing=0
     monospacedWidth=1
     size=24
     skewAmount=0
     typefaceFontName=Black
     outlineSettings {
      bApplyOutlineToDropShadows=false
      bMiteredCorners=false
      bSeparateFillAlpha=false
      outlineMaterial=null
      outlineSize=0
      outlineColor=<rgba 0 0 0 1>
     }
    }
    shadowOffset=<0 0>
   }
   slot {class=SizeBoxSlot}
  }
  HorizontalBox_126 {
   widgetClassPath=&HorizontalBox
   parent=Overlay_0
   widget=HorizontalBox_126
   slot {class=OverlaySlot, properties {horizontalAlignment=HAlign_Right, verticalAlignment=VAlign_Center}}
  }
  Button_EditLinks {
   widgetClassPath=/Game/Athena/HUD/Bacchus/HUDLayoutTool/HUDLayoutToolV2_Button
   bInherited=true
   parent=HorizontalBox_126
   widget=Button_EditLinks
   properties {
    bOverride_HeightOverride=true
    bOverride_WidthOverride=true
    buttonHeightOverride=80
    buttonWidthOverride=260
    text=Edit Links
    visibility=Visible
    iconBrush {
     bIsDynamicallyLoaded=false
     drawAs=Image
     imageType=NoImage
     mirroring=NoMirror
     resourceName=None
     resourceObject=/Game/Athena/HUD/Bacchus/HUDLayoutTool/HLT_Icons/T-HLT-Link_Sprite
     tiling=NoTile
     imageSize=<60 30>
     margin=<ltrb 0 0 0 0>
     outlineSettings {
      bUseBrushTransparency=false
      roundingType=HalfHeightRadius
      width=0
      color {colorUseRule=UseColor_Specified, specifiedColor=<rgba 0 0 0 0>}
      cornerRadii=<0 0 0 0>
     }
     tintColor {colorUseRule=UseColor_Specified, specifiedColor=<rgba 1 1 1 1>}
     uVRegion {bIsValid=false, max=<0 0>, min=<0 0>}
    }
    iconPadding=<ltrb 10 0 0 0>
    internalPadding=<ltrb 0 0 0 0>
   }
   slot {class=HorizontalBoxSlot, properties {padding=<ltrb 20 0 0 0>}}
  }
  Button_RevertHUDWidget {
   widgetClassPath=&HUDLayoutToolV2_Button
   bInherited=true
   parent=HorizontalBox_126
   widget=Button_RevertHUDWidget
   properties {
    bOverride_HeightOverride=true
    bOverride_WidthOverride=true
    buttonHeightOverride=80
    buttonWidthOverride=90
    text=""
    visibility=Visible
    iconBrush {
     bIsDynamicallyLoaded=false
     drawAs=Image
     imageType=NoImage
     mirroring=NoMirror
     resourceName=None
     resourceObject=/Game/Athena/HUD/Bacchus/HUDLayoutTool/HLT_Icons/T-HLT-ResetV2_Sprite
     tiling=NoTile
     imageSize=<60 30>
     margin=<ltrb 0 0 0 0>
     outlineSettings {
      bUseBrushTransparency=false
      roundingType=HalfHeightRadius
      width=0
      color {colorUseRule=UseColor_Specified, specifiedColor=<rgba 0 0 0 0>}
      cornerRadii=<0 0 0 0>
     }
     tintColor {colorUseRule=UseColor_Specified, specifiedColor=<rgba 1 1 1 1>}
     uVRegion {bIsValid=false, max=<0 0>, min=<0 0>}
    }
    internalPadding=<ltrb 0 0 0 0>
   }
   slot {class=HorizontalBoxSlot, properties {padding=<ltrb 20 0 0 0>}}
  }
  Button_DeleteHUDWidget {
   widgetClassPath=&HUDLayoutToolV2_Button
   bInherited=true
   parent=HorizontalBox_126
   widget=Button_DeleteHUDWidget
   properties {
    bOverride_HeightOverride=true
    bOverride_WidthOverride=true
    buttonHeightOverride=80
    buttonWidthOverride=90
    text=""
    visibility=Visible
    iconBrush {
     bIsDynamicallyLoaded=false
     drawAs=Image
     imageType=NoImage
     mirroring=NoMirror
     resourceName=None
     resourceObject=/Game/Athena/HUD/Atlas/Bacchus/Delete_Normal_Sprite
     tiling=NoTile
     imageSize=<52 60>
     margin=<ltrb 0 0 0 0>
     outlineSettings {
      bUseBrushTransparency=false
      roundingType=HalfHeightRadius
      width=0
      color {colorUseRule=UseColor_Specified, specifiedColor=<rgba 0 0 0 0>}
      cornerRadii=<0 0 0 0>
     }
     tintColor {colorUseRule=UseColor_Specified, specifiedColor=<rgba 1 1 1 1>}
     uVRegion {bIsValid=false, max=<0 0>, min=<0 0>}
    }
    internalPadding=<ltrb 0 0 0 0>
    normalButtonStyle {
     iconAngle=0
     backgroundColor=<rgba 0.6939 0.0137 0.016 1>
     iconColor=<rgba 1 1 1 1>
     textColor {colorUseRule=UseColor_Specified, specifiedColor=<rgba 1 1 1 1>}
    }
    pressedButtonStyle {
     iconAngle=0
     backgroundColor=<rgba 0.7529 0.0723 0.0762 1>
     iconColor=<rgba 1 1 1 1>
     textColor {colorUseRule=UseColor_Specified, specifiedColor=<rgba 1 1 1 1>}
    }
   }
   slot {class=HorizontalBoxSlot, properties {padding=<ltrb 20 0 0 0>}}
  }
  ListView_Settings {
   widgetClassPath=/Script/FortniteUI.FortSettingsListView
   bInherited=true
   bIsVariable=true
   parent=VerticalBox_PropertiesContainer
   widget=ListView_Settings
   properties {
    entryWidgetClass=&/HUDLayoutToolV2_EmptySetting.HUDLayoutToolV2_EmptySetting_C
    horizontalEntrySpacing=25
    numDesignerPreviewEntries=11
    verticalEntrySpacing=25
    visualData=&/HUDLayoutToolV2_VisualSettingsRegistry
   }
   slot {class=VerticalBoxSlot, properties {padding=<ltrb 0 6 0 0>, size {sizeRule=Fill, value=1}}}
  }
 ]
}
