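# Control panel widget for the mobile HUD Layout Tool V2, laying out the controls for editing HUD layouts.
# A handle strip tops a header with the HUD Layouts profile button and Save, Revert All and Exit icon buttons.
# Toggleable Modes and Properties buttons share a row with grid snap options, above a switcher holding the context selector and property panel.
# The event graph is empty, and most controls bind to the native FortMobileHUDLayoutToolV2_ControlPanel parent.
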WidgetBlueprint /FortniteCoreUI/HUD/Bacchus/HUDLayoutTool/HUDLayoutToolV2_ControlPanel
graphs [EventGraph {links [], nodes []}]
variables []
widgets {
 info {
  inheritedWidgetCount=10
  parentClass=/Script/FortniteUI.FortMobileHUDLayoutToolV2_ControlPanel
  rootWidgetClass=&Overlay
  widgetCount=25
 }
 widgets [
  Overlay_0 {widgetClassPath=&Overlay, widget=Overlay_0}
  Background {
   widgetClassPath=&Image
   parent=Overlay_0
   widget=Background
   properties {
    brush {
     bIsDynamicallyLoaded=false
     drawAs=Image
     imageType=NoImage
     mirroring=NoMirror
     resourceName=None
     resourceObject=null
     tiling=NoTile
     imageSize=<64 64>
     margin=<ltrb 0 0 0 0>
     outlineSettings {
      bUseBrushTransparency=false
      roundingType=HalfHeightRadius
      width=0
      color {colorUseRule=UseColor_Specified, specifiedColor=<rgba 0 0 0 0>}
      cornerRadii=<0 0 0 0>
     }
     tintColor {colorUseRule=UseColor_Specified, specifiedColor=<rgba 1 1 1 1>}
     uVRegion {bIsValid=false, max=<0 0>, min=<0 0>}
    }
    colorAndOpacity=<rgba 0.0235 0.0235 0.0235 1>
   }
   slot {class=OverlaySlot, properties {horizontalAlignment=HAlign_Fill, verticalAlignment=VAlign_Fill}}
  }
  VerticalBox_240 {
   widgetClassPath=&VerticalBox
   parent=Overlay_0
   widget=VerticalBox_240
   slot {class=OverlaySlot, properties {horizontalAlignment=HAlign_Fill, verticalAlignment=VAlign_Fill}}
  }
  Overlay_69 {
   widgetClassPath=&Overlay
   parent=VerticalBox_240
   widget=Overlay_69
   slot {class=VerticalBoxSlot, properties {padding=<ltrb 20 0 20 0>}}
  }
  Border_87 {
   widgetClassPath=&Border
   parent=Overlay_69
   widget=Border_87
   properties {
    visibility=SelfHitTestInvisible
    background {
     bIsDynamicallyLoaded=false
     drawAs=Image
     imageType=NoImage
     mirroring=NoMirror
     resourceName=None
     resourceObject=null
     tiling=NoTile
     imageSize=<32 32>
     margin=<ltrb 0 0 0 0>
     outlineSettings {
      bUseBrushTransparency=false
      roundingType=HalfHeightRadius
      width=0
      color {colorUseRule=UseColor_Specified, specifiedColor=<rgba 0 0 0 0>}
      cornerRadii=<0 0 0 0>
     }
     tintColor {colorUseRule=UseColor_Specified, specifiedColor=<rgba 1 1 1 0.3>}
     uVRegion {bIsValid=false, max=<0 0>, min=<0 0>}
    }
   }
   slot {class=OverlaySlot, properties {horizontalAlignment=HAlign_Fill, verticalAlignment=VAlign_Center}}
  }
  Image_PanelHandle {
   widgetClassPath=&Image
   bIsVariable=true
   parent=Overlay_69
   widget=Image_PanelHandle
   properties {
    visibility=SelfHitTestInvisible
    brush {
     bIsDynamicallyLoaded=false
     drawAs=Image
     imageType=NoImage
     mirroring=NoMirror
     resourceName=None
     resourceObject=/FortniteCoreUI/Athena/HUD/Bacchus/HUDLayoutTool/HLT_Icons/MI_HUDLayoutToolV2_PanelHandle
     tiling=NoTile
     imageSize=<50 50>
     margin=<ltrb 0 0 0 0>
     outlineSettings {
      bUseBrushTransparency=false
      roundingType=HalfHeightRadius
      width=0
      color {colorUseRule=UseColor_Specified, specifiedColor=<rgba 0 0 0 0>}
      cornerRadii=<0 0 0 0>
     }
     tintColor {colorUseRule=UseColor_Specified, specifiedColor=<rgba 1 1 1 1>}
     uVRegion {bIsValid=false, max=<0 0>, min=<0 0>}
    }
   }
   slot {class=OverlaySlot, properties {horizontalAlignment=HAlign_Center, verticalAlignment=VAlign_Center}}
  }
  VerticalBox_3 {widgetClassPath=&VerticalBox, parent=VerticalBox_240, widget=VerticalBox_3, slot {class=VerticalBoxSlot}}
  SizeBox_0 {widgetClassPath=&SizeBox, parent=VerticalBox_3, widget=SizeBox_0, properties {heightOverride=100}, slot {class=VerticalBoxSlot}}
  Overlay_3 {widgetClassPath=&Overlay, parent=SizeBox_0, widget=Overlay_3, slot {class=SizeBoxSlot, properties {padding=<ltrb 20 0 20 10>}}}
  Overlay_2 {
   widgetClassPath=&Overlay
   parent=Overlay_3
   widget=Overlay_2
   slot {class=OverlaySlot, properties {verticalAlignment=VAlign_Fill}}
  }
  Button_OpenProfileSelectionPanel {
   widgetClassPath=/Game/Athena/HUD/Bacchus/HUDLayoutTool/HUDLayoutToolV2_Button
   bInherited=true
   parent=Overlay_2
   widget=Button_OpenProfileSelectionPanel
   properties {
    bOverride_HeightOverride=true
    bOverride_WidthOverride=true
    buttonHeightOverride=80
    buttonWidthOverride=500
    text=HUD Layouts
    textFontSize=26
    internalPadding=<ltrb 0 0 0 0>
    textPadding=<ltrb 0 0 206 0>
   }
   slot {class=OverlaySlot}
  }
  ButtonDotIcon {
   widgetClassPath=&Image
   bIsVariable=true
   parent=Overlay_2
   widget=ButtonDotIcon
   properties {brush {
     bIsDynamicallyLoaded=false
     drawAs=Image
     imageType=NoImage
     mirroring=NoMirror
     resourceName=None
     resourceObject=/Game/Athena/HUD/Bacchus/HUDLayoutTool/HLT_Icons/T-HLT-Dots_Sprite
     tiling=NoTile
     imageSize=<12 46>
     margin=<ltrb 0 0 0 0>
     outlineSettings {
      bUseBrushTransparency=false
      roundingType=HalfHeightRadius
      width=0
      color {colorUseRule=UseColor_Specified, specifiedColor=<rgba 0 0 0 0>}
      cornerRadii=<0 0 0 0>
     }
     tintColor {colorUseRule=UseColor_Specified, specifiedColor=<rgba 1 1 1 1>}
     uVRegion {bIsValid=false, max=<0 0>, min=<0 0>}
    }}
   slot {class=OverlaySlot, properties {horizontalAlignment=HAlign_Right, verticalAlignment=VAlign_Center, padding=<ltrb 0 0 30 0>}}
  }
  HorizontalBox_2 {
   widgetClassPath=&HorizontalBox
   parent=Overlay_3
   widget=HorizontalBox_2
   slot {class=OverlaySlot, properties {horizontalAlignment=HAlign_Right, verticalAlignment=VAlign_Fill}}
  }
  Button_Save {
   widgetClassPath=&HUDLayoutToolV2_Button
   bInherited=true
   parent=HorizontalBox_2
   widget=Button_Save
   properties {
    bOverride_HeightOverride=true
    bOverride_WidthOverride=true
    buttonHeightOverride=80
    buttonWidthOverride=90
    text=""
    iconBrush {
     bIsDynamicallyLoaded=false
     drawAs=Image
     imageType=NoImage
     mirroring=NoMirror
     resourceName=None
     resourceObject=/Game/Athena/HUD/Bacchus/HUDLayoutTool/HLT_Icons/T-HLT-Save_Sprite
     tiling=NoTile
     imageSize=<46 50>
     margin=<ltrb 0 0 0 0>
     outlineSettings {
      bUseBrushTransparency=false
      roundingType=HalfHeightRadius
      width=0
      color {colorUseRule=UseColor_Specified, specifiedColor=<rgba 0 0 0 0>}
      cornerRadii=<0 0 0 0>
     }
     tintColor {colorUseRule=UseColor_Specified, specifiedColor=<rgba 1 1 1 1>}
     uVRegion {bIsValid=false, max=<0 0>, min=<0 0>}
    }
    iconPadding=<ltrb 20 0 20 0>
    internalPadding=<ltrb 0 0 0 0>
   }
   slot {class=HorizontalBoxSlot, properties {padding=<ltrb 0 0 20 0>}}
  }
  Button_RevertAll {
   widgetClassPath=&HUDLayoutToolV2_Button
   bInherited=true
   parent=HorizontalBox_2
   widget=Button_RevertAll
   properties {
    bOverride_HeightOverride=true
    bOverride_WidthOverride=true
    buttonHeightOverride=80
    buttonWidthOverride=90
    text=""
    iconBrush {
     bIsDynamicallyLoaded=false
     drawAs=Image
     imageType=NoImage
     mirroring=NoMirror
     resourceName=None
     resourceObject=/Game/Athena/HUD/Bacchus/HUDLayoutTool/HLT_Icons/T-HLT-ResetV2_Sprite
     tiling=NoTile
     imageSize=<60 30>
     margin=<ltrb 0 0 0 0>
     outlineSettings {
      bUseBrushTransparency=false
      roundingType=HalfHeightRadius
      width=0
      color {colorUseRule=UseColor_Specified, specifiedColor=<rgba 0 0 0 0>}
      cornerRadii=<0 0 0 0>
     }
     tintColor {colorUseRule=UseColor_Specified, specifiedColor=<rgba 1 1 1 1>}
     uVRegion {bIsValid=false, max=<0 0>, min=<0 0>}
    }
    internalPadding=<ltrb 0 0 0 0>
   }
   slot {class=HorizontalBoxSlot, properties {padding=<ltrb 0 0 20 0>}}
  }
  Button_Exit {
   widgetClassPath=&HUDLayoutToolV2_Button
   bInherited=true
   parent=HorizontalBox_2
   widget=Button_Exit
   properties {
    bOverride_HeightOverride=true
    bOverride_WidthOverride=true
    buttonHeightOverride=80
    buttonWidthOverride=90
    text=""
    iconBrush {
     bIsDynamicallyLoaded=false
     drawAs=Image
     imageType=NoImage
     mirroring=NoMirror
     resourceName=None
     resourceObject=/Game/Athena/HUD/Bacchus/HUDLayoutTool/HLT_Icons/T-HLT-Exit_Sprite
     tiling=NoTile
     imageSize=<40 60>
     margin=<ltrb 0 0 0 0>
     outlineSettings {
      bUseBrushTransparency=false
      roundingType=HalfHeightRadius
      width=0
      color {colorUseRule=UseColor_Specified, specifiedColor=<rgba 0 0 0 0>}
      cornerRadii=<0 0 0 0>
     }
     tintColor {colorUseRule=UseColor_Specified, specifiedColor=<rgba 1 1 1 1>}
     uVRegion {bIsValid=false, max=<0 0>, min=<0 0>}
    }
    internalPadding=<ltrb 0 0 0 0>
    normalButtonStyle {
     iconAngle=0
     backgroundColor=<rgba 0.6939 0.0137 0.016 1>
     iconColor=<rgba 1 1 1 1>
     textColor {colorUseRule=UseColor_Specified, specifiedColor=<rgba 1 1 1 1>}
    }
    pressedButtonStyle {
     iconAngle=0
     backgroundColor=<rgba 0.7529 0.0723 0.0762 1>
     iconColor=<rgba 1 1 1 1>
     textColor {colorUseRule=UseColor_Specified, specifiedColor=<rgba 1 1 1 1>}
    }
   }
   slot {class=HorizontalBoxSlot}
  }
  HorizontalBox_0 {widgetClassPath=&HorizontalBox, parent=VerticalBox_3, widget=HorizontalBox_0, slot {class=VerticalBoxSlot}}
  Button_SwitchContexts {
   widgetClassPath=&HUDLayoutToolV2_Button
   bInherited=true
   bIsVariable=true
   parent=HorizontalBox_0
   widget=Button_SwitchContexts
   properties {
    bIsToggleable=true
    bOverride_HeightOverride=true
    bOverride_WidthOverride=true
    buttonHeightOverride=80
    buttonWidthOverride=240
    text=Modes
    textFontSize=26
    internalPadding=<ltrb 0 0 0 0>
    normalButtonStyle {
     iconAngle=0
     backgroundColor=<rgba 0 0.117 1 0.5>
     iconColor=<rgba 1 1 1 0.5>
     textColor {colorUseRule=UseColor_Specified, specifiedColor=<rgba 1 1 1 0.5>}
    }
    selectedButtonStyle {
     iconAngle=0
     backgroundColor=<rgba 0 0.1137 1 1>
     iconColor=<rgba 1 1 1 1>
     textColor {colorUseRule=UseColor_Specified, specifiedColor=<rgba 1 1 1 1>}
    }
   }
   slot {class=HorizontalBoxSlot, properties {padding=<ltrb 20 10 10 20>}}
  }
  Button_SwitchProperties {
   widgetClassPath=&HUDLayoutToolV2_Button
   bInherited=true
   bIsVariable=true
   parent=HorizontalBox_0
   widget=Button_SwitchProperties
   properties {
    bIsToggleable=true
    bOverride_HeightOverride=true
    bOverride_WidthOverride=true
    buttonHeightOverride=80
    buttonWidthOverride=240
    text=Properties
    textFontSize=26
    internalPadding=<ltrb 0 0 0 0>
    normalButtonStyle {
     iconAngle=0
     backgroundColor=<rgba 0 0.117 1 0.5>
     iconColor=<rgba 1 1 1 0.5>
     textColor {colorUseRule=UseColor_Specified, specifiedColor=<rgba 1 1 1 0.5>}
    }
    selectedButtonStyle {
     iconAngle=0
     backgroundColor=<rgba 0 0.1137 1 1>
     iconColor=<rgba 1 1 1 1>
     textColor {colorUseRule=UseColor_Specified, specifiedColor=<rgba 1 1 1 1>}
    }
   }
   slot {class=HorizontalBoxSlot, properties {padding=<ltrb 10 10 10 20>}}
  }
  GridSnapOption_Controls {
   widgetClassPath=/FortniteCoreUI/HUD/Bacchus/HUDLayoutTool/ControlPanel/HUDLayoutToolV2_GridSnapOption
   bInherited=true
   parent=HorizontalBox_0
   widget=GridSnapOption_Controls
   slot {class=HorizontalBoxSlot, properties {size {sizeRule=Fill, value=1}}}
  }
  Overlay_1 {
   widgetClassPath=&Overlay
   parent=VerticalBox_240
   widget=Overlay_1
   slot {class=VerticalBoxSlot, properties {size {sizeRule=Fill, value=1}}}
  }
  Separator {
   widgetClassPath=&Border
   parent=Overlay_1
   widget=Separator
   properties {
    background {
     bIsDynamicallyLoaded=false
     drawAs=Image
     imageType=NoImage
     mirroring=NoMirror
     resourceName=None
     resourceObject=null
     tiling=NoTile
     imageSize=<32 32>
     margin=<ltrb 0 0 0 0>
     outlineSettings {
      bUseBrushTransparency=false
      roundingType=HalfHeightRadius
      width=0
      color {colorUseRule=UseColor_Specified, specifiedColor=<rgba 0 0 0 0>}
      cornerRadii=<0 0 0 0>
     }
     tintColor {colorUseRule=UseColor_Specified, specifiedColor=<rgba 1 1 1 0.3>}
     uVRegion {bIsValid=false, max=<0 0>, min=<0 0>}
    }
    padding=<ltrb 0 2 0 2>
   }
   slot {class=OverlaySlot, properties {horizontalAlignment=HAlign_Fill, padding=<ltrb 20 0 20 0>}}
  }
  Switcher_Controls {
   widgetClassPath=&CommonActivatableWidgetSwitcher
   bInherited=true
   parent=Overlay_1
   widget=Switcher_Controls
   properties {transitionDuration=0.3}
   slot {class=OverlaySlot, properties {horizontalAlignment=HAlign_Fill, verticalAlignment=VAlign_Fill, padding=<ltrb 0 5 0 0>}}
  }
  ContextSelector_ContextLayouts {
   widgetClassPath=/FortniteCoreUI/Athena/HUD/Bacchus/HUDLayoutTool/ControlPanel/HUDLayoutToolV2_ContextSelector
   bInherited=true
   parent=Switcher_Controls
   widget=ContextSelector_ContextLayouts
   slot {class=WidgetSwitcherSlot}
  }
  PropertyPanel_WidgetProperties {
   widgetClassPath=/FortniteCoreUI/HUD/Bacchus/HUDLayoutTool/ControlPanel/HUDLayoutToolV2_PropertyPanel
   bInherited=true
   parent=Switcher_Controls
   widget=PropertyPanel_WidgetProperties
   slot {class=WidgetSwitcherSlot}
  }
 ]
}
