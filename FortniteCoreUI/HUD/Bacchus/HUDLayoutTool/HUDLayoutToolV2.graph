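# Widget Blueprint layout for the mobile HUD Layout Tool, where HUD widgets are arranged; all logic lives outside it, with FortMobileHUDLayoutToolV2 as parent.
# A virtual canvas stacks background, grid and HUD proxy layer beneath a top control panel and bottom widget catalog, both with toggle buttons.
# Two short animations slide the panel and catalog; the panel's also fades Exit.
# Profile, link-across, toast and tooltip overlays start collapsed.
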WidgetBlueprint /FortniteCoreUI/HUD/Bacchus/HUDLayoutTool/HUDLayoutToolV2
animations [
 Anim_ControlPanelToggle {
  binding_count=4
  length_seconds=0.15
  object=&self:Anim_ControlPanelToggle
  sequence {bindings [
    HorizontalBoxSlot_1 {
     id=#1
     parent_id=#2
     possessed_class=HorizontalBoxSlot
     tracks [{
       type=MovieSceneMarginTrack
       display_name=Padding
       sections [{
         type=MovieSceneMarginSection
         end_frame=3
         start_frame=0
         channels [
          Left {type=MovieSceneScriptingFloatChannel, keys [[0.0003,3.4],[2.978,-100]]}
          Top {type=MovieSceneScriptingFloatChannel}
          Right {type=MovieSceneScriptingFloatChannel}
          Bottom {type=MovieSceneScriptingFloatChannel}
         ]
        }]
      }]
    }
    VerticalBox_ControlPanelContainer {id=#3, possessed_class=VerticalBox, tracks []}
    CanvasPanelSlot_0 {
     id=#4
     parent_id=#3
     possessed_class=CanvasPanelSlot
     tracks [{
       type=MovieSceneMarginTrack
       display_name="Offsets (Layout Data)"
       sections [{
         type=MovieSceneMarginSection
         end_frame=3
         start_frame=0
         channels [
          Left {type=MovieSceneScriptingFloatChannel}
          Top {type=MovieSceneScriptingFloatChannel, keys [[0.0003,-660],[3,0]]}
          Right {type=MovieSceneScriptingFloatChannel}
          Bottom {type=MovieSceneScriptingFloatChannel}
         ]
        }]
      }]
    }
    Button_Exit {
     id=#2
     possessed_class=HUDLayoutToolV2_Button_C
     tracks [
      {
       type=MovieSceneByteTrack
       display_name=Visibility
       sections [{
         type=MovieSceneByteSection
         end_frame=3
         start_frame=0
         channels [None {type=MovieSceneScriptingByteChannel, keys [[0.0003,0],[0.138,4],[2.9243,1]]}]
        }]
      }
      {
       type=MovieSceneFloatTrack
       display_name=Render Opacity
       sections [{
         type=MovieSceneFloatSection
         end_frame=3
         start_frame=0
         channels [None {type=MovieSceneScriptingFloatChannel, keys [[0.0003,1],[1.5123,0]]}]
        }]
      }
     ]
    }
   ]}
 }
 Anim_ToggleWidgetCatalog {
  binding_count=4
  length_seconds=0.15
  object=&self:Anim_ToggleWidgetCatalog
  sequence {bindings [
    WidgetCatalog_Catalog {id=#5, possessed_class=HUDLayoutToolV2_WidgetCatalog_C, tracks []}
    CanvasPanelSlot_1 {
     id=#6
     parent_id=#7
     possessed_class=CanvasPanelSlot
     tracks [{
       type=MovieSceneMarginTrack
       display_name="Offsets (Layout Data)"
       sections [{
         type=MovieSceneMarginSection
         end_frame=3
         start_frame=0
         channels [
          Left {type=MovieSceneScriptingFloatChannel}
          Top {type=MovieSceneScriptingFloatChannel, keys [[0,0],[3,-235]]}
          Right {type=MovieSceneScriptingFloatChannel}
          Bottom {type=MovieSceneScriptingFloatChannel}
         ]
        }]
      }]
    }
    CanvasPanelSlot_1 {
     id=#8
     parent_id=#5
     possessed_class=CanvasPanelSlot
     tracks [{
       type=MovieSceneMarginTrack
       display_name="Offsets (Layout Data)"
       sections [{
         type=MovieSceneMarginSection
         end_frame=3
         start_frame=0
         channels [
          Left {type=MovieSceneScriptingFloatChannel}
          Top {type=MovieSceneScriptingFloatChannel, keys [[0,400],[3,0]]}
          Right {type=MovieSceneScriptingFloatChannel}
          Bottom {type=MovieSceneScriptingFloatChannel}
         ]
        }]
      }]
    }
    SafeZoneToggleWidgetCatalog {id=#7, possessed_class=SafeZone, tracks []}
   ]}
 }
]
graphs [EventGraph {links [], nodes []}]
variables []
widgets {
 info {inheritedWidgetCount=21, parentClass=/Script/FortniteUI.FortMobileHUDLayoutToolV2, rootWidgetClass=&Overlay, widgetCount=36}
 widgets [
  Overlay_ToolMainOverlay {widgetClassPath=&Overlay, bInherited=true, widget=Overlay_ToolMainOverlay}
  BackgroundSafeZone {
   widgetClassPath=&Image
   parent=Overlay_ToolMainOverlay
   widget=BackgroundSafeZone
   properties {
    brush {
     bIsDynamicallyLoaded=false
     drawAs=Image
     imageType=NoImage
     mirroring=NoMirror
     resourceName=None
     resourceObject=/FortniteCoreUI/Athena/HUD/Bacchus/HUDLayoutTool/HLT_Icons/M_HLT_RadialGradientRed_Inst
     tiling=NoTile
     imageSize=<32 32>
     margin=<ltrb 0 0 0 0>
     outlineSettings {
      bUseBrushTransparency=false
      roundingType=HalfHeightRadius
      width=0
      color {colorUseRule=UseColor_Specified, specifiedColor=<rgba 0 0 0 0>}
      cornerRadii=<0 0 0 0>
     }
     tintColor {colorUseRule=UseColor_Specified, specifiedColor=<rgba 1 1 1 1>}
     uVRegion {bIsValid=false, max=<0 0>, min=<0 0>}
    }
    colorAndOpacity=<rgba 0.151 0.151 0.151 1>
   }
   slot {class=OverlaySlot, properties {horizontalAlignment=HAlign_Fill, verticalAlignment=VAlign_Fill}}
  }
  SafeZone_2 {
   widgetClassPath=&SafeZone
   bIsVariable=true
   parent=Overlay_ToolMainOverlay
   widget=SafeZone_2
   slot {class=OverlaySlot, properties {horizontalAlignment=HAlign_Fill, verticalAlignment=VAlign_Fill}}
  }
  Background {
   widgetClassPath=&Image
   parent=SafeZone_2
   widget=Background
   properties {brush {
     bIsDynamicallyLoaded=false
     drawAs=Image
     imageType=NoImage
     mirroring=NoMirror
     resourceName=None
     resourceObject=&M_HLT_RadialGradientRed_Inst
     tiling=NoTile
     imageSize=<32 32>
     margin=<ltrb 0 0 0 0>
     outlineSettings {
      bUseBrushTransparency=false
      roundingType=HalfHeightRadius
      width=0
      color {colorUseRule=UseColor_Specified, specifiedColor=<rgba 0 0 0 0>}
      cornerRadii=<0 0 0 0>
     }
     tintColor {colorUseRule=UseColor_Specified, specifiedColor=<rgba 1 1 1 1>}
     uVRegion {bIsValid=false, max=<0 0>, min=<0 0>}
    }}
   slot {class=SafeZoneSlot}
  }
  VirtualCanvas_WidgetLayer {
   widgetClassPath=/Script/FortniteUI.HUDLayoutToolV2_VirtualCanvas
   bInherited=true
   bIsVariable=true
   parent=Overlay_ToolMainOverlay
   widget=VirtualCanvas_WidgetLayer
   slot {class=OverlaySlot, properties {horizontalAlignment=HAlign_Fill, verticalAlignment=VAlign_Fill}}
  }
  SafeZoneHUDLayer {
   widgetClassPath=&SafeZone
   bIsVariable=true
   parent=VirtualCanvas_WidgetLayer
   widget=SafeZoneHUDLayer
   slot {class=HUDLayoutToolV2_VirtualCanvasSlot}
  }
  Overlay_VirtualCanvasWorkspace {
   widgetClassPath=&Overlay
   bInherited=true
   parent=SafeZoneHUDLayer
   widget=Overlay_VirtualCanvasWorkspace
   slot {class=SafeZoneSlot}
  }
  ForegroundOutline {
   widgetClassPath=&Image
   parent=Overlay_VirtualCanvasWorkspace
   widget=ForegroundOutline
   properties {
    brush {
     bIsDynamicallyLoaded=false
     drawAs=Border
     imageType=NoImage
     mirroring=NoMirror
     resourceName=None
     resourceObject=/Game/Athena/HUD/Bacchus/HUDLayoutTool/HLT_Icons/M_UI_Square_Inst
     tiling=NoTile
     imageSize=<32 32>
     margin=<ltrb 0.45 0.45 0.45 0.45>
     outlineSettings {
      bUseBrushTransparency=false
      roundingType=HalfHeightRadius
      width=0
      color {colorUseRule=UseColor_Specified, specifiedColor=<rgba 0 0 0 0>}
      cornerRadii=<0 0 0 0>
     }
     tintColor {colorUseRule=UseColor_Specified, specifiedColor=<rgba 1 1 1 1>}
     uVRegion {bIsValid=false, max=<0 0>, min=<0 0>}
    }
    colorAndOpacity=<rgba 0.9301 1 0.0097 1>
   }
   slot {class=OverlaySlot, properties {horizontalAlignment=HAlign_Fill, verticalAlignment=VAlign_Fill, padding=<ltrb -4 -4 -4 -4>}}
  }
  Foreground {
   widgetClassPath=&Image
   parent=Overlay_VirtualCanvasWorkspace
   widget=Foreground
   properties {brush {
     bIsDynamicallyLoaded=false
     drawAs=Image
     imageType=NoImage
     mirroring=NoMirror
     resourceName=None
     resourceObject=/FortniteCoreUI/Athena/HUD/Bacchus/HUDLayoutTool/HLT_Icons/MI-HLT-Background
     tiling=NoTile
     imageSize=<32 32>
     margin=<ltrb 0 0 0 0>
     outlineSettings {
      bUseBrushTransparency=false
      roundingType=HalfHeightRadius
      width=0
      color {colorUseRule=UseColor_Specified, specifiedColor=<rgba 0 0 0 0>}
      cornerRadii=<0 0 0 0>
     }
     tintColor {colorUseRule=UseColor_Specified, specifiedColor=<rgba 1 1 1 1>}
     uVRegion {bIsValid=false, max=<0 0>, min=<0 0>}
    }}
   slot {class=OverlaySlot, properties {horizontalAlignment=HAlign_Fill, verticalAlignment=VAlign_Fill}}
  }
  Grid_Visual {
   widgetClassPath=/Script/FortniteUI.HUDLayoutToolV2_Grid
   bInherited=true
   bIsVariable=true
   parent=Overlay_VirtualCanvasWorkspace
   widget=Grid_Visual
   properties {ruleLinePeriod=4, visibility=SelfHitTestInvisible, regularLineColor=<rgba 1 1 1 0.1>, ruleLineColor=<rgba 1 1 1 0.25>}
   slot {class=OverlaySlot, properties {horizontalAlignment=HAlign_Fill, verticalAlignment=VAlign_Fill}}
  }
  HUDLayer_HUDWidgetProxies {
   widgetClassPath=/Script/FortniteUI.FortMobileHUDLayer
   bInherited=true
   bIsVariable=true
   parent=Overlay_VirtualCanvasWorkspace
   widget=HUDLayer_HUDWidgetProxies
   properties {visibility=SelfHitTestInvisible}
   slot {class=OverlaySlot, properties {horizontalAlignment=HAlign_Fill, verticalAlignment=VAlign_Fill}}
  }
  SafeZoneResetVirtualCanvas {
   widgetClassPath=&SafeZone
   parent=Overlay_ToolMainOverlay
   widget=SafeZoneResetVirtualCanvas
   properties {padBottom=false}
   slot {class=OverlaySlot, properties {horizontalAlignment=HAlign_Fill, verticalAlignment=VAlign_Fill}}
  }
  Button_ResetVirtualCanvas {
   widgetClassPath=/Game/Athena/HUD/Bacchus/HUDLayoutTool/HUDLayoutToolV2_Button
   bInherited=true
   bIsVariable=true
   parent=SafeZoneResetVirtualCanvas
   widget=Button_ResetVirtualCanvas
   properties {
    bOverride_HeightOverride=true
    bOverride_WidthOverride=true
    buttonHeightOverride=96
    buttonWidthOverride=108
    text=""
    visibility=Collapsed
    iconBrush {
     bIsDynamicallyLoaded=false
     drawAs=Image
     imageType=NoImage
     mirroring=NoMirror
     resourceName=None
     resourceObject=/Game/Athena/HUD/Bacchus/HUDLayoutTool/HLT_Icons/T-HLT-CenterCanvas_Sprite
     tiling=NoTile
     imageSize=<64 64>
     margin=<ltrb 0 0 0 0>
     outlineSettings {
      bUseBrushTransparency=false
      roundingType=HalfHeightRadius
      width=0
      color {colorUseRule=UseColor_Specified, specifiedColor=<rgba 0 0 0 0>}
      cornerRadii=<0 0 0 0>
     }
     tintColor {colorUseRule=UseColor_Specified, specifiedColor=<rgba 1 1 1 1>}
     uVRegion {bIsValid=false, max=<0 0>, min=<0 0>}
    }
    internalPadding=<ltrb 0 0 0 0>
   }
   slot {class=SafeZoneSlot, properties {hAlign=HAlign_Left, vAlign=VAlign_Top}}
  }
  CanvasPanel_ControlsContainer {
   widgetClassPath=&CanvasPanel
   bInherited=true
   parent=Overlay_ToolMainOverlay
   widget=CanvasPanel_ControlsContainer
   slot {class=OverlaySlot, properties {horizontalAlignment=HAlign_Fill, verticalAlignment=VAlign_Fill}}
  }
  VerticalBox_ControlPanelContainer {
   widgetClassPath=&VerticalBox
   bInherited=true
   parent=CanvasPanel_ControlsContainer
   widget=VerticalBox_ControlPanelContainer
   slot {
    class=CanvasPanelSlot
    properties {bAutoSize=true, layoutData {alignment=<0.5 0>, anchors {maximum=<0.5 0>, minimum=<0.5 0>}, offsets=<ltrb 0 0 0 0>}}
   }
  }
  Sizebox {
   widgetClassPath=&SizeBox
   parent=VerticalBox_ControlPanelContainer
   widget=Sizebox
   properties {
    bOverride_MaxDesiredHeight=true
    bOverride_MinDesiredHeight=true
    bOverride_WidthOverride=true
    heightOverride=660
    maxDesiredHeight=660
    minDesiredHeight=646
    widthOverride=934
   }
   slot {class=VerticalBoxSlot}
  }
  ControlPanel_Controls {
   widgetClassPath=&/HUDLayoutToolV2_ControlPanel
   bInherited=true
   bIsVariable=true
   parent=Sizebox
   widget=ControlPanel_Controls
   properties {visibility=Visible}
   slot {class=SizeBoxSlot}
  }
  Overlay_0 {widgetClassPath=&Overlay, parent=VerticalBox_ControlPanelContainer, widget=Overlay_0, slot {class=VerticalBoxSlot}}
  VerticalBox_2 {
   widgetClassPath=&VerticalBox
   parent=Overlay_0
   widget=VerticalBox_2
   slot {class=OverlaySlot, properties {padding=<ltrb 10 4 0 0>}}
  }
  SizeBox_ProfileName {
   widgetClassPath=&SizeBox
   parent=VerticalBox_2
   widget=SizeBox_ProfileName
   properties {bOverride_WidthOverride=true, visibility=HitTestInvisible, widthOverride=366}
   slot {class=VerticalBoxSlot}
  }
  Text_ProfileName {
   widgetClassPath=&CommonTextBlock
   bInherited=true
   parent=SizeBox_ProfileName
   widget=Text_ProfileName
   properties {
    justification=Right
    scrollStyle=&TextScrollStyle-Base_C
    text=Profile Name
    colorAndOpacity {colorUseRule=UseColor_Specified, specifiedColor=<rgba 0.0232 0.0232 0.0232 0.8>}
    font {
     bForceMonospaced=false
     bMaterialIsStencil=false
     fontMaterial=null
     fontObject=&BurbankSmall
     letterSpacing=0
     monospacedWidth=1
     size=24
     skewAmount=0
     typefaceFontName=Black
     outlineSettings {
      bApplyOutlineToDropShadows=false
      bMiteredCorners=false
      bSeparateFillAlpha=false
      outlineMaterial=null
      outlineSize=0
      outlineColor=<rgba 0 0 0 1>
     }
    }
    shadowOffset=<0 0>
   }
   slot {class=SizeBoxSlot}
  }
  SizeBox_WidgetName {
   widgetClassPath=&SizeBox
   parent=VerticalBox_2
   widget=SizeBox_WidgetName
   properties {bOverride_WidthOverride=true, visibility=HitTestInvisible, widthOverride=366}
   slot {class=VerticalBoxSlot}
  }
  Text_WidgetName {
   widgetClassPath=&CommonTextBlock
   bInherited=true
   parent=SizeBox_WidgetName
   widget=Text_WidgetName
   properties {
    justification=Right
    scrollStyle=&TextScrollStyle-Base_C
    colorAndOpacity {colorUseRule=UseColor_Specified, specifiedColor=<rgba 0.0232 0.0232 0.0232 0.8>}
    font {
     bForceMonospaced=false
     bMaterialIsStencil=false
     fontMaterial=null
     fontObject=&BurbankSmall
     letterSpacing=0
     monospacedWidth=1
     size=20
     skewAmount=0
     typefaceFontName=Black
     outlineSettings {
      bApplyOutlineToDropShadows=false
      bMiteredCorners=false
      bSeparateFillAlpha=false
      outlineMaterial=null
      outlineSize=0
      outlineColor=<rgba 0 0 0 1>
     }
    }
    shadowOffset=<0 0>
   }
   slot {class=SizeBoxSlot}
  }
  SizeBox_ContextName {
   widgetClassPath=&SizeBox
   parent=Overlay_0
   widget=SizeBox_ContextName
   properties {bOverride_WidthOverride=true, widthOverride=366}
   slot {class=OverlaySlot, properties {horizontalAlignment=HAlign_Right, verticalAlignment=VAlign_Fill, padding=<ltrb 0 0 10 0>}}
  }
  HorizontalBox_1 {widgetClassPath=&HorizontalBox, parent=SizeBox_ContextName, widget=HorizontalBox_1, slot {class=SizeBoxSlot}}
  Button_Exit {
   widgetClassPath=&HUDLayoutToolV2_Button
   bInherited=true
   parent=HorizontalBox_1
   widget=Button_Exit
   properties {
    bOverride_HeightOverride=true
    bOverride_WidthOverride=true
    buttonHeightOverride=86
    buttonWidthOverride=90
    text=""
    iconBrush {
     bIsDynamicallyLoaded=false
     drawAs=Image
     imageType=NoImage
     mirroring=NoMirror
     resourceName=None
     resourceObject=/Game/Athena/HUD/Bacchus/HUDLayoutTool/HLT_Icons/T-HLT-Exit_Sprite
     tiling=NoTile
     imageSize=<40 60>
     margin=<ltrb 0 0 0 0>
     outlineSettings {
      bUseBrushTransparency=false
      roundingType=HalfHeightRadius
      width=0
      color {colorUseRule=UseColor_Specified, specifiedColor=<rgba 0 0 0 0>}
      cornerRadii=<0 0 0 0>
     }
     tintColor {colorUseRule=UseColor_Specified, specifiedColor=<rgba 1 1 1 1>}
     uVRegion {bIsValid=false, max=<0 0>, min=<0 0>}
    }
    internalPadding=<ltrb 0 0 0 0>
    normalButtonStyle {
     iconAngle=0
     backgroundColor=<rgba 0.6939 0.0137 0.016 1>
     iconColor=<rgba 1 1 1 1>
     textColor {colorUseRule=UseColor_Specified, specifiedColor=<rgba 1 1 1 1>}
    }
    pressedButtonStyle {
     iconAngle=0
     backgroundColor=<rgba 0.7529 0.0723 0.0762 1>
     iconColor=<rgba 1 1 1 1>
     textColor {colorUseRule=UseColor_Specified, specifiedColor=<rgba 1 1 1 1>}
    }
   }
   slot {class=HorizontalBoxSlot, properties {horizontalAlignment=HAlign_Center, verticalAlignment=VAlign_Top, padding=<ltrb 3.4 0 10 0>}}
  }
  Text_ContextName {
   widgetClassPath=&CommonTextBlock
   bInherited=true
   parent=HorizontalBox_1
   widget=Text_ContextName
   properties {
    scrollStyle=&TextScrollStyle-Base_C
    text=Context Name
    colorAndOpacity {colorUseRule=UseColor_Specified, specifiedColor=<rgba 0.0232 0.0232 0.0232 0.8>}
    font {
     bForceMonospaced=false
     bMaterialIsStencil=false
     fontMaterial=null
     fontObject=&BurbankSmall
     letterSpacing=0
     monospacedWidth=1
     size=24
     skewAmount=0
     typefaceFontName=Black
     outlineSettings {
      bApplyOutlineToDropShadows=false
      bMiteredCorners=false
      bSeparateFillAlpha=false
      outlineMaterial=null
      outlineSize=0
      outlineColor=<rgba 0 0 0 1>
     }
    }
    shadowOffset=<0 0>
   }
   slot {class=HorizontalBoxSlot, properties {verticalAlignment=VAlign_Top, padding=<ltrb 0 4 0 0>, size {sizeRule=Fill, value=1}}}
  }
  Button_ToggleControlPanel {
   widgetClassPath=&HUDLayoutToolV2_Button
   bInherited=true
   bIsVariable=true
   parent=Overlay_0
   widget=Button_ToggleControlPanel
   properties {
    bIsToggleable=true
    bOverride_HeightOverride=true
    buttonHeightOverride=86
    text=""
    iconBrush {
     bIsDynamicallyLoaded=false
     drawAs=Image
     imageType=NoImage
     mirroring=NoMirror
     resourceName=None
     resourceObject=/Game/Athena/HUD/Bacchus/HUDLayoutTool/HLT_Icons/T-HLT-Arrow_Sprite
     tiling=NoTile
     imageSize=<51 37>
     margin=<ltrb 0 0 0 0>
     outlineSettings {
      bUseBrushTransparency=false
      roundingType=HalfHeightRadius
      width=0
      color {colorUseRule=UseColor_Specified, specifiedColor=<rgba 0 0 0 0>}
      cornerRadii=<0 0 0 0>
     }
     tintColor {colorUseRule=UseColor_Specified, specifiedColor=<rgba 1 1 1 1>}
     uVRegion {bIsValid=false, max=<0 0>, min=<0 0>}
    }
    normalButtonStyle {
     iconAngle=180
     backgroundColor=<rgba 0 0.117 1 1>
     iconColor=<rgba 1 1 1 1>
     textColor {colorUseRule=UseColor_Specified, specifiedColor=<rgba 1 1 1 1>}
    }
    selectedButtonStyle {
     iconAngle=0
     backgroundColor=<rgba 0 0.117 1 1>
     iconColor=<rgba 1 1 1 1>
     textColor {colorUseRule=UseColor_Specified, specifiedColor=<rgba 1 1 1 1>}
    }
   }
   slot {class=OverlaySlot, properties {horizontalAlignment=HAlign_Center}}
  }
  WidgetCatalog_Catalog {
   widgetClassPath=/FortniteCoreUI/Athena/HUD/Bacchus/HUDLayoutTool/HUDLayoutToolV2_WidgetCatalog
   bInherited=true
   parent=CanvasPanel_ControlsContainer
   widget=WidgetCatalog_Catalog
   slot {
    class=CanvasPanelSlot
    properties {bAutoSize=true, layoutData {alignment=<0.5 1>, anchors {maximum=<1 1>, minimum=<0 1>}, offsets=<ltrb 0 0 0 0>}}
   }
  }
  SafeZoneToggleWidgetCatalog {
   widgetClassPath=&SafeZone
   parent=CanvasPanel_ControlsContainer
   widget=SafeZoneToggleWidgetCatalog
   slot {
    class=CanvasPanelSlot
    properties {bAutoSize=true, layoutData {alignment=<1 1>, anchors {maximum=<1 1>, minimum=<1 1>}, offsets=<ltrb 0 0 0 0>}}
   }
  }
  Button_ToggleWidgetCatalog {
   widgetClassPath=&HUDLayoutToolV2_Button
   bInherited=true
   parent=SafeZoneToggleWidgetCatalog
   widget=Button_ToggleWidgetCatalog
   properties {
    bIsToggleable=true
    bOverride_HeightOverride=true
    buttonHeightOverride=86
    text=""
    iconBrush {
     bIsDynamicallyLoaded=false
     drawAs=Image
     imageType=NoImage
     mirroring=NoMirror
     resourceName=None
     resourceObject=&T-HLT-Arrow_Sprite
     tiling=NoTile
     imageSize=<51 37>
     margin=<ltrb 0 0 0 0>
     outlineSettings {
      bUseBrushTransparency=false
      roundingType=HalfHeightRadius
      width=0
      color {colorUseRule=UseColor_Specified, specifiedColor=<rgba 0 0 0 0>}
      cornerRadii=<0 0 0 0>
     }
     tintColor {colorUseRule=UseColor_Specified, specifiedColor=<rgba 1 1 1 1>}
     uVRegion {bIsValid=false, max=<0 0>, min=<0 0>}
    }
    selectedButtonStyle {
     iconAngle=180
     backgroundColor=<rgba 0 0.117 1 1>
     iconColor=<rgba 1 1 1 1>
     textColor {colorUseRule=UseColor_Specified, specifiedColor=<rgba 1 1 1 1>}
    }
   }
   slot {class=SafeZoneSlot}
  }
  Image_WidgetAnchorPoint {
   widgetClassPath=&Image
   bInherited=true
   bIsVariable=true
   parent=CanvasPanel_ControlsContainer
   widget=Image_WidgetAnchorPoint
   properties {
    visibility=Collapsed
    brush {
     bIsDynamicallyLoaded=false
     drawAs=Image
     imageType=FullColor
     mirroring=NoMirror
     resourceName=None
     resourceObject=&T-HLT-CenterCanvas_Sprite
     tiling=NoTile
     imageSize=<64 64>
     margin=<ltrb 0 0 0 0>
     outlineSettings {
      bUseBrushTransparency=false
      roundingType=HalfHeightRadius
      width=0
      color {colorUseRule=UseColor_Specified, specifiedColor=<rgba 0 0 0 0>}
      cornerRadii=<0 0 0 0>
     }
     tintColor {colorUseRule=UseColor_Specified, specifiedColor=<rgba 1 1 1 1>}
     uVRegion {bIsValid=false, max=<0 0>, min=<0 0>}
    }
   }
   slot {
    class=CanvasPanelSlot
    properties {layoutData {alignment=<0.5 0.5>, anchors {maximum=<0.5 0.5>, minimum=<0.5 0.5>}, offsets=<ltrb 0 0 30 30>}}
   }
  }
  Toaster_Notifications {
   widgetClassPath=/FortniteCoreUI/Athena/HUD/Bacchus/HUDLayoutTool/HUDLayoutToolV2_Toaster
   bInherited=true
   parent=Overlay_ToolMainOverlay
   widget=Toaster_Notifications
   properties {visibility=Collapsed}
   slot {class=OverlaySlot, properties {padding=<ltrb 0 200 0 0>}}
  }
  ProfileSelectionPanel_Profiles {
   widgetClassPath=/FortniteCoreUI/Athena/HUD/Bacchus/HUDLayoutTool/HUDLayoutToolV2_ProfileSelectionPanel
   bInherited=true
   parent=Overlay_ToolMainOverlay
   widget=ProfileSelectionPanel_Profiles
   properties {visibility=Collapsed}
   slot {class=OverlaySlot, properties {horizontalAlignment=HAlign_Fill, verticalAlignment=VAlign_Fill}}
  }
  LinkAcrossPanel_LinkManagementWindow {
   widgetClassPath=/FortniteCoreUI/Athena/HUD/Bacchus/HUDLayoutTool/LinkAcross/HUDLayoutToolV2_LinkAcrossPanel
   bInherited=true
   bIsVariable=true
   parent=Overlay_ToolMainOverlay
   widget=LinkAcrossPanel_LinkManagementWindow
   properties {visibility=Collapsed}
   slot {class=OverlaySlot, properties {horizontalAlignment=HAlign_Fill, verticalAlignment=VAlign_Fill}}
  }
  TooltipLayer_Messages {
   widgetClassPath=/FortniteCoreUI/Athena/HUD/Bacchus/HUDLayoutTool/HUDLayoutToolV2_TooltipLayer
   bInherited=true
   bIsVariable=true
   parent=Overlay_ToolMainOverlay
   widget=TooltipLayer_Messages
   properties {visibility=Collapsed}
   slot {class=OverlaySlot, properties {horizontalAlignment=HAlign_Fill, verticalAlignment=VAlign_Fill}}
  }
 ]
}
